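# Physics asset for the rabbit on the QuaintToy female medium body cosmetic, giving it limited secondary motion and collision.
# Capsules cover the rabbit's head, body, limbs and ears; constraints link limbs to body, body to head, and each ear to a small base-ear body.
# All joints lock twist and set narrow swing limits.
# Front-left skirt bones carry multi-capsule shapes with no constraints, apparently as colliders for the rabbit.
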PhysicsAsset /BRCosmetics/Characters/Player/Female/Medium/Bodies/F_MED_QuaintToy/Meshes/F_MED_QuaintToy_RabbitPhysics
bodies {
 base_rabbitEar_l [{shapeType=Capsule, length=0.505, radius=0.505, shapeName=base_rabbitEar_l_capsule, rotation=<rot 0 -90 0>}]
 base_rabbitEar_r [{shapeType=Capsule, length=0.505, radius=0.505, shapeName=base_rabbitEar_r_capsule, rotation=<rot 0 -90 0>}]
 dyn_rabbitArm_l [{
   shapeType=Capsule
   length=4.0215
   radius=1.5907
   shapeName=dyn_rabbitArm_l_capsule
   center=<2.7593 -0.134 0.0719>
   rotation=<rot 0 -96.1564 -88.8664>
  }]
 dyn_rabbitArm_r [{
   shapeType=Capsule
   length=4.2632
   radius=1.5819
   shapeName=dyn_rabbitArm_r_capsule
   center=<2.6906 0.0658 -0.1623>
   rotation=<rot 0 84.5886 -88.7584>
  }]
 dyn_rabbitBody [{
   shapeType=Capsule
   length=3.0945
   radius=3.1896
   shapeName=dyn_rabbitBody_capsule
   center=<4.759 -1.1571 0.0092>
   rotation=<rot 0.0324 90.207 -89.753>
  }]
 dyn_rabbitEar_01_l [{
   shapeType=Capsule
   length=2.3041
   radius=1.0049
   shapeName=dyn_rabbitEar_01_l_capsule
   center=<1.8145 -0.0935 -0.0889>
   rotation=<rot -0.6041 97.6383 -86.8758>
  }]
 dyn_rabbitEar_01_r [{
   shapeType=Capsule
   length=2.4932
   radius=1.1321
   shapeName=dyn_rabbitEar_01_r_capsule
   center=<1.5309 -0.0879 -0.0895>
   rotation=<rot 0 -89.3833 -88.7697>
  }]
 dyn_rabbitHead [{
   shapeType=Capsule
   length=0.0081
   radius=4.0269
   shapeName=dyn_rabbitHead_capsule
   center=<3.4374 -1.6876 -0.0337>
   rotation=<rot -0.5466 -90 3.0402>
  }]
 dyn_rabbitLeg_l [{
   shapeType=Capsule
   length=4.3011
   radius=1.7178
   shapeName=dyn_rabbitLeg_r_capsule
   center=<2.9398 -0.2559 0.1043>
   rotation=<rot 0.0001 89.043 -84.2809>
  }]
 dyn_rabbitLeg_r [{
   shapeType=Capsule
   length=4.3011
   radius=1.7178
   shapeName=dyn_rabbitLeg_r_capsule
   center=<2.9531 -0.3789 -0.049>
   rotation=<rot 0.0001 89.043 -84.2809>
  }]
 dyn_skirt_04_fr_l [
  {
   shapeType=Capsule
   length=6.0109
   radius=7.8375
   shapeName=dyn_skirt_05_fr_l_capsule
   center=<4.8176 7.6435 -4.7011>
   rotation=<rot 7.7627 -89.996 -3.4097>
  }
  {shapeType=Capsule, length=3.8628, radius=3.9367, shapeName=dyn_skirt_04_fr_l_capsule, center=<0 2.7511 0.2929>, rotation=<rot 0 0 7.4242>}
 ]
 dyn_skirt_05_fr_l [
  {
   shapeType=Capsule
   length=5.8754
   radius=7.8375
   shapeName=dyn_skirt_05_fr_l_capsule
   center=<4.564 6.9876 -0.3149>
   rotation=<rot 7.8302 -90 -3.4103>
  }
  {
   shapeType=Capsule
   length=1.9703
   radius=2.435
   shapeName=dyn_skirt_05_fr_l_capsule
   center=<8.6356 0.1077 -4.1819>
   rotation=<rot 0 -0.0001 51.6137>
  }
  {
   shapeType=Capsule
   length=15
   radius=2.747
   shapeName=dyn_skirt_05_fr_l_capsule
   center=<-0.4662 2.5207 -0.0095>
   rotation=<rot 0.0001 -0.0002 5.4663>
  }
  {
   shapeType=Capsule
   length=1.9703
   radius=2.435
   shapeName=dyn_skirt_05_fr_l_capsule
   center=<5.8868 0.4367 -4.5972>
   rotation=<rot 0 -0.0001 51.6137>
  }
 ]
}
constraints [
 {bone1Name=dyn_rabbitArm_r, bone2Name=dyn_rabbitBody, swing1LimitDegrees=15, swing2LimitDegrees=10, twistLimitDegrees=45, twistMotion=Locked}
 {bone1Name=dyn_rabbitArm_l, bone2Name=dyn_rabbitBody, swing1LimitDegrees=15, swing2LimitDegrees=10, twistLimitDegrees=45, twistMotion=Locked}
 {bone1Name=dyn_rabbitLeg_l, bone2Name=dyn_rabbitBody, swing1LimitDegrees=30, swing2LimitDegrees=10, twistLimitDegrees=45, twistMotion=Locked}
 {bone1Name=dyn_rabbitLeg_r, bone2Name=dyn_rabbitBody, swing1LimitDegrees=30, swing2LimitDegrees=10, twistLimitDegrees=45, twistMotion=Locked}
 {bone1Name=dyn_rabbitBody, bone2Name=dyn_rabbitHead, swing1LimitDegrees=5, swing2LimitDegrees=8, twistLimitDegrees=45, twistMotion=Locked}
 {
  bone1Name=dyn_rabbitEar_01_r
  bone2Name=base_rabbitEar_r
  swing1LimitDegrees=10
  swing2LimitDegrees=12
  twistLimitDegrees=45
  twistMotion=Locked
 }
 {
  bone1Name=dyn_rabbitEar_01_l
  bone2Name=base_rabbitEar_l
  swing1LimitDegrees=10
  swing2LimitDegrees=12
  twistLimitDegrees=45
  twistMotion=Locked
 }
]
