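# Physics asset for the skirt on the female medium QuaintToy player cosmetic body.
# Five three-segment capsule chains (front, front-left, front-right, left, right) are constrained to spine_02 and then link to link downward, with twist locked and tight swing limits.
# Unconstrained capsules on the pelvis, spine, glute deform bones, thighs and calves give body volumes, and the front-left tip carries extra angled capsules.
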PhysicsAsset /BRCosmetics/Characters/Player/Female/Medium/Bodies/F_MED_QuaintToy/Meshes/F_MED_QuaintToy_SkirtPhysics
bodies {
 calf_l [{shapeType=Capsule, length=44.0142, radius=5.7312, center=<-20.8813 -2.6223 -0.2696>, rotation=<rot -0.4264 93.0743 89.6341>}]
 calf_r [{shapeType=Capsule, length=44.0142, radius=5.7312, center=<20.8813 2.6223 0.2696>, rotation=<rot -0.4264 93.0743 -90.3659>}]
 deform_glute_l [{
   shapeType=Capsule
   length=2.5157
   radius=0.849
   shapeName=deform_glute_l_capsule
   center=<3.2235 1.6505 3.374>
   rotation=<rot -0.4962 -74.497 166.0954>
  }]
 deform_glute_r [{
   shapeType=Capsule
   length=2.5157
   radius=0.849
   shapeName=deform_glute_r_capsule
   center=<-3.2235 -1.6505 -3.374>
   rotation=<rot -0.4962 -74.497 -13.9046>
  }]
 dyn_skirt_03_fr [{shapeType=Capsule, length=10.7873, radius=6.4175, shapeName=dyn_skirt_01_fr_capsule, center=<5.4285 -1.1188 0>, rotation=<rot 0 90 0>}]
 dyn_skirt_03_fr_l [{shapeType=Capsule, length=10.7873, radius=6.4175, shapeName=dyn_skirt_01_fr_capsule, center=<5.4285 -3.2517 0>, rotation=<rot 0 90 0>}]
 dyn_skirt_03_fr_r [{shapeType=Capsule, length=10.7873, radius=6.4175, shapeName=dyn_skirt_01_fr_capsule, center=<5.4285 -3.5826 0>, rotation=<rot 0 90 0>}]
 dyn_skirt_03_l [{shapeType=Capsule, length=10.7873, radius=6.4175, shapeName=dyn_skirt_01_fr_capsule, center=<5.4285 -4.1774 0>, rotation=<rot 0 90 0>}]
 dyn_skirt_03_r [{shapeType=Capsule, length=10.7873, radius=6.4175, shapeName=dyn_skirt_01_fr_capsule, center=<5.4285 -4.4932 0>, rotation=<rot 0 90 0>}]
 dyn_skirt_04_fr [{shapeType=Capsule, length=10.7873, radius=6.4175, shapeName=dyn_skirt_01_fr_capsule, center=<5.4285 0.1277 0>, rotation=<rot 0 90 0>}]
 dyn_skirt_04_fr_l [{shapeType=Capsule, length=10.7873, radius=6.4175, shapeName=dyn_skirt_01_fr_capsule, center=<5.4285 -3.185 0>, rotation=<rot 0 90 0>}]
 dyn_skirt_04_fr_r [{shapeType=Capsule, length=10.7873, radius=6.4175, shapeName=dyn_skirt_01_fr_capsule, center=<5.4285 -2.5655 0>, rotation=<rot 0 90 0>}]
 dyn_skirt_04_l [{shapeType=Capsule, length=10.7873, radius=6.4175, shapeName=dyn_skirt_01_fr_capsule, center=<5.4285 -2.9129 0>, rotation=<rot 0 90 0>}]
 dyn_skirt_04_r [{shapeType=Capsule, length=10.7873, radius=6.4175, shapeName=dyn_skirt_01_fr_capsule, center=<5.4285 -3.3914 0>, rotation=<rot 0 90 0>}]
 dyn_skirt_05_fr [{shapeType=Capsule, length=8.9336, radius=6.4175, shapeName=dyn_skirt_01_fr_capsule, center=<8.413 1.9914 0>, rotation=<rot 0 90 0>}]
 dyn_skirt_05_fr_l [
  {shapeType=Capsule, length=14.265, radius=6.4175, shapeName=dyn_skirt_01_fr_capsule, center=<8.413 -5.2505 0>, rotation=<rot 0.1514 90 0>}
  {
   shapeType=Capsule
   length=10.2121
   radius=5.9538
   shapeName=dyn_skirt_05_fr_l_capsule
   center=<5.9479 -0.384 -10.9118>
   rotation=<rot -8.4195 5.7813 -45.1115>
  }
  {
   shapeType=Capsule
   length=10
   radius=5.9538
   shapeName=dyn_skirt_05_fr_l_capsule
   center=<7.2766 0.2249 8.62>
   rotation=<rot 7.5975 15.0392 45.5719>
  }
 ]
 dyn_skirt_05_fr_r [{shapeType=Capsule, length=12.7059, radius=6.4175, shapeName=dyn_skirt_01_fr_capsule, center=<8.413 0.1962 0>, rotation=<rot 0 90 0>}]
 dyn_skirt_05_l [{
   shapeType=Capsule
   length=13.4742
   radius=6.4175
   shapeName=dyn_skirt_01_fr_capsule
   center=<8.413 1.6558 0.1996>
   rotation=<rot -19.828 89.9999 0.0003>
  }]
 dyn_skirt_05_r [{shapeType=Capsule, length=10.7873, radius=6.4175, shapeName=dyn_skirt_01_fr_capsule, center=<8.413 1.5288 0>, rotation=<rot 0 90 0>}]
 pelvis [{shapeType=Capsule, length=14.7584, radius=9.4159, center=<-0.7329 -3.1911 0.0752>, rotation=<rot 0 90 0>}]
 spine_01 [{shapeType=Capsule, length=10.0165, radius=8.2164, center=<-0.7092 -0.6244 0.1079>, rotation=<rot 0 90 0>}]
 spine_02 [{shapeType=Capsule, length=6.3648, radius=4.9185, center=<-0.7329 -0.2313 0.0752>, rotation=<rot 0 90 0>}]
 thigh_l [{shapeType=Capsule, length=38.171, radius=7.6443, center=<-23.3717 -1.3062 0.462>, rotation=<rot 0 -90.4611 90.9689>}]
 thigh_r [{shapeType=Capsule, length=38.171, radius=7.6443, center=<23.367 0.7159 -0.462>, rotation=<rot 0 -90.4611 -89.0311>}]
}
constraints [
 {bone1Name=dyn_skirt_03_fr, bone2Name=spine_02, swing1LimitDegrees=5, swing2LimitDegrees=3, twistLimitDegrees=45, twistMotion=Locked}
 {bone1Name=dyn_skirt_04_fr, bone2Name=dyn_skirt_03_fr, swing1LimitDegrees=7, swing2LimitDegrees=3, twistLimitDegrees=45, twistMotion=Locked}
 {bone1Name=dyn_skirt_05_fr, bone2Name=dyn_skirt_04_fr, swing1LimitDegrees=8, swing2LimitDegrees=3, twistLimitDegrees=45, twistMotion=Locked}
 {bone1Name=dyn_skirt_03_fr_l, bone2Name=spine_02, swing1LimitDegrees=6, swing2LimitDegrees=5, twistLimitDegrees=45, twistMotion=Locked}
 {
  bone1Name=dyn_skirt_04_fr_l
  bone2Name=dyn_skirt_03_fr_l
  swing1LimitDegrees=6
  swing2LimitDegrees=5
  twistLimitDegrees=45
  twistMotion=Locked
 }
 {
  bone1Name=dyn_skirt_05_fr_l
  bone2Name=dyn_skirt_04_fr_l
  swing1LimitDegrees=6
  swing2LimitDegrees=5
  twistLimitDegrees=45
  twistMotion=Locked
 }
 {bone1Name=dyn_skirt_03_fr_r, bone2Name=spine_02, swing1LimitDegrees=5, swing2LimitDegrees=3, twistLimitDegrees=45, twistMotion=Locked}
 {
  bone1Name=dyn_skirt_04_fr_r
  bone2Name=dyn_skirt_03_fr_r
  swing1LimitDegrees=7
  swing2LimitDegrees=3
  twistLimitDegrees=45
  twistMotion=Locked
 }
 {
  bone1Name=dyn_skirt_05_fr_r
  bone2Name=dyn_skirt_04_fr_r
  swing1LimitDegrees=8
  swing2LimitDegrees=3
  twistLimitDegrees=45
  twistMotion=Locked
 }
 {bone1Name=dyn_skirt_03_l, bone2Name=spine_02, swing1LimitDegrees=5, swing2LimitDegrees=4, twistLimitDegrees=45, twistMotion=Locked}
 {bone1Name=dyn_skirt_04_l, bone2Name=dyn_skirt_03_l, swing1LimitDegrees=10, swing2LimitDegrees=4, twistLimitDegrees=45, twistMotion=Locked}
 {bone1Name=dyn_skirt_05_l, bone2Name=dyn_skirt_04_l, swing1LimitDegrees=10, swing2LimitDegrees=4, twistLimitDegrees=45, twistMotion=Locked}
 {bone1Name=dyn_skirt_03_r, bone2Name=spine_02, swing1LimitDegrees=5, swing2LimitDegrees=4, twistLimitDegrees=45, twistMotion=Locked}
 {bone1Name=dyn_skirt_04_r, bone2Name=dyn_skirt_03_r, swing1LimitDegrees=10, swing2LimitDegrees=4, twistLimitDegrees=45, twistMotion=Locked}
 {bone1Name=dyn_skirt_05_r, bone2Name=dyn_skirt_04_r, swing1LimitDegrees=10, swing2LimitDegrees=4, twistLimitDegrees=45, twistMotion=Locked}
]
