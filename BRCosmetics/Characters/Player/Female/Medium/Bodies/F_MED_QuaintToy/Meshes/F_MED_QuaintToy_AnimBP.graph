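# Animation Blueprint for the F_MED_QuaintToy female body cosmetic, adding secondary motion to its skirt, toy rabbit, ribbons, bows, sleeves and bobby pin.
# It copies the attached parent's pose into a Control Rig driven by movement and emote flags.
# RigidBody skirt and rabbit simulations interleave with chained AnimDynamics springs, then bone offsets follow, some blended by skydiving, walking or tactical sprint.
# A separate rabbit AnimDynamics chain has no input pose connected.
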AnimBlueprint /BRCosmetics/Characters/Player/Female/Medium/Bodies/F_MED_QuaintToy/Meshes/F_MED_QuaintToy_AnimBP
target_skeleton=&/F_MED_QuaintToy_Skeleton
graphs [
 AnimGraph {nodes [
   n1 = Anim Dynamics - Bone: dyn_rabbitLeg_l!(ComponentPose, Alpha=0.8, GravityOverride:Vector="0,0,-980", ExternalForce="0,0,0") -> (Pose) {
    angularDampingOverride=0.999
    angularSpringConstant=3000
    bAngularSpring=true
    bOverrideAngularDamping=true
    bUseGravityOverride=true
    gravityScale=0.35
    simulationSpace=RootRelative
    boundBone {boneName=dyn_rabbitLeg_l}
    componentAppliedLinearAccClamp=<3000 3000 3000>
    componentLinearAccScale=<0.3 0.3 0.3>
    componentLinearVelScale=<0.3 0.3 0.3>
    gravityOverride {z=-980}
    physicsBodyDefinitions [{
      boundBone {boneName=dyn_rabbitLeg_l}
      boxExtents=<35 35 35>
      constraintSetup {angularLimitsMax=<0 10 3>, angularLimitsMin=<0 -10 0>, angularTarget=<1 0 0>}
      localJointOffset=<3 3 0>
     }]
    relativeSpaceBone {boneName=spine_05}
    simSpaceSettings {simSpaceAngularAlpha=0.1}
   }
   n2 = Anim Dynamics - Bone: dyn_rabbitLeg_r!(ComponentPose=@n1, Alpha=0.8, GravityOverride:Vector="0,0,-980", ExternalForce="0,0,0") -> (Pose) {
    angularDampingOverride=0.999
    angularSpringConstant=3000
    bAngularSpring=true
    bOverrideAngularDamping=true
    bUseGravityOverride=true
    gravityScale=0.35
    simulationSpace=RootRelative
    boundBone {boneName=dyn_rabbitLeg_r}
    componentAppliedLinearAccClamp=<3000 3000 3000>
    componentLinearAccScale=<0.3 0.3 0.3>
    componentLinearVelScale=<0.3 0.3 0.3>
    gravityOverride {z=-980}
    physicsBodyDefinitions [{
      boundBone {boneName=dyn_rabbitLeg_r}
      boxExtents=<35 35 35>
      constraintSetup {angularLimitsMax=<0 10 3>, angularLimitsMin=<0 -10 0>, angularTarget=<1 0 0>}
      localJointOffset=<3 3 0>
     }]
    relativeSpaceBone {boneName=spine_05}
    simSpaceSettings {simSpaceAngularAlpha=0.1}
   }
   n3 = Anim Dynamics - Bone: dyn_rabbitArm_l!(ComponentPose=@n2, Alpha=0.8, GravityOverride:Vector="0,0,-980", ExternalForce="0,0,0") -> (Pose) {
    angularDampingOverride=0.999
    angularSpringConstant=3000
    bAngularSpring=true
    bOverrideAngularDamping=true
    bUseGravityOverride=true
    gravityScale=0.35
    simulationSpace=RootRelative
    boundBone {boneName=dyn_rabbitArm_l}
    componentAppliedLinearAccClamp=<3000 3000 3000>
    componentLinearAccScale=<0.3 0.3 0.3>
    componentLinearVelScale=<0.3 0.3 0.3>
    gravityOverride {z=-980}
    physicsBodyDefinitions [{
      boundBone {boneName=dyn_rabbitArm_l}
      boxExtents=<35 35 35>
      constraintSetup {angularLimitsMax=<0 10 2>, angularLimitsMin=<0 -10 0>, angularTarget=<1 0 0>}
      localJointOffset=<3 3 0>
     }]
    relativeSpaceBone {boneName=spine_05}
    simSpaceSettings {simSpaceAngularAlpha=0.1}
   }
   n4 = Anim Dynamics - Bone: dyn_rabbitArm_r!(ComponentPose=@n3, Alpha=0.8, GravityOverride:Vector="0,0,-980", ExternalForce="0,0,0") -> (Pose) {
    angularDampingOverride=0.999
    angularSpringConstant=3000
    bAngularSpring=true
    bOverrideAngularDamping=true
    bUseGravityOverride=true
    gravityScale=0.35
    simulationSpace=RootRelative
    boundBone {boneName=dyn_rabbitArm_r}
    componentAppliedLinearAccClamp=<3000 3000 3000>
    componentLinearAccScale=<0.3 0.3 0.3>
    componentLinearVelScale=<0.3 0.3 0.3>
    gravityOverride {z=-980}
    physicsBodyDefinitions [{
      boundBone {boneName=dyn_rabbitArm_r}
      boxExtents=<35 35 35>
      constraintSetup {angularLimitsMax=<0 10 2>, angularLimitsMin=<0 -10 0>, angularTarget=<1 0 0>}
      localJointOffset=<3 3 0>
     }]
    relativeSpaceBone {boneName=spine_05}
    simSpaceSettings {simSpaceAngularAlpha=0.1}
   }
   n5 = Anim Dynamics - Bone: dyn_rabbitEar_01_r!(ComponentPose=@n4, Alpha=0.8, GravityOverride:Vector="0,0,-980", ExternalForce="0,0,0") -> (Pose) {
    angularDampingOverride=0.999
    angularSpringConstant=3000
    bAngularSpring=true
    bOverrideAngularDamping=true
    bUseGravityOverride=true
    gravityScale=0.35
    simulationSpace=RootRelative
    boundBone {boneName=dyn_rabbitEar_01_r}
    componentAppliedLinearAccClamp=<3000 3000 3000>
    componentLinearAccScale=<0.3 0.3 0.3>
    componentLinearVelScale=<0.3 0.3 0.3>
    gravityOverride {z=-980}
    physicsBodyDefinitions [{
      boundBone {boneName=dyn_rabbitEar_01_r}
      boxExtents=<35 35 35>
      constraintSetup {angularLimitsMax=<0 1 0>, angularLimitsMin=<0 -15 0>, angularTarget=<1 0 0>}
      localJointOffset=<2 -1 0>
     }]
    relativeSpaceBone {boneName=spine_05}
    simSpaceSettings {simSpaceAngularAlpha=0.1}
   }
   n6 = Anim Dynamics - Bone: dyn_rabbitEar_01_l!(ComponentPose=@n5, Alpha=0.8, GravityOverride:Vector="0,0,-980", ExternalForce="0,0,0") -> (Pose) {
    angularDampingOverride=0.999
    angularSpringConstant=3000
    bAngularSpring=true
    bOverrideAngularDamping=true
    bUseGravityOverride=true
    gravityScale=0.35
    simulationSpace=RootRelative
    boundBone {boneName=dyn_rabbitEar_01_l}
    componentAppliedLinearAccClamp=<3000 3000 3000>
    componentLinearAccScale=<0.3 0.3 0.3>
    componentLinearVelScale=<0.3 0.3 0.3>
    gravityOverride {z=-980}
    physicsBodyDefinitions [{
      boundBone {boneName=dyn_rabbitEar_01_l}
      boxExtents=<35 35 35>
      constraintSetup {angularLimitsMax=<0 15 0>, angularLimitsMin=<0 -1 0>, angularTarget=<1 0 0>}
      localJointOffset=<2 -1 0>
     }]
    relativeSpaceBone {boneName=spine_05}
    simSpaceSettings {simSpaceAngularAlpha=0.1}
   }
   n7 = Anim Dynamics - Bone: base_rabbitBody!(ComponentPose=@n6, Alpha=0.8, GravityOverride:Vector="0,0,-980", ExternalForce="0,0,0") -> (Pose) {
    angularDampingOverride=0.999
    angularSpringConstant=3000
    bAngularSpring=true
    bOverrideAngularDamping=true
    bUseGravityOverride=true
    gravityScale=0.35
    simulationSpace=RootRelative
    boundBone {boneName=base_rabbitBody}
    componentAppliedLinearAccClamp=<3000 3000 3000>
    componentLinearAccScale=<0.3 0.3 0.3>
    componentLinearVelScale=<0.3 0.3 0.3>
    gravityOverride {z=-980}
    physicsBodyDefinitions [{
      boundBone {boneName=base_rabbitBody}
      boxExtents=<35 35 35>
      constraintSetup {angularLimitsMax=<0 10 2>, angularLimitsMin=<0 -10 -3>, angularTarget=<1 0 0>}
      localJointOffset=<5 -1 0>
     }]
    relativeSpaceBone {boneName=spine_05}
    simSpaceSettings {simSpaceAngularAlpha=0.1}
   }
   Anim Dynamics - Bone: dyn_rabbitHead!(ComponentPose=@n7, Alpha=0.8, GravityOverride:Vector="0,0,-980", ExternalForce="0,0,0") -> (Pose) {
    angularDampingOverride=0.999
    angularSpringConstant=3000
    bAngularSpring=true
    bOverrideAngularDamping=true
    bUseGravityOverride=true
    gravityScale=0.35
    simulationSpace=RootRelative
    boundBone {boneName=dyn_rabbitHead}
    componentAppliedLinearAccClamp=<3000 3000 3000>
    componentLinearAccScale=<0.3 0.3 0.3>
    componentLinearVelScale=<0.3 0.3 0.3>
    gravityOverride {z=-980}
    physicsBodyDefinitions [{
      boundBone {boneName=dyn_rabbitHead}
      boxExtents=<35 35 35>
      constraintSetup {angularLimitsMax=<0 10 0>, angularLimitsMin=<0 -10 0>, angularTarget=<1 0 0>}
      localJointOffset=<2 -1 0>
     }]
    relativeSpaceBone {boneName=spine_05}
    simSpaceSettings {simSpaceAngularAlpha=0.1}
   }
   n8 = Control Rig!(
    Source=Copy Pose From Mesh() {bUseAttachedParent=true}
    Alpha
    isCrouching_CR:bool=Get bIsCrouching()
    isBaseballSliding_CR:bool=Get bIsBaseballSliding()
    isKneeSliding_CR:bool=Get bIsKneeSliding()
    isPlayingEmote_CR:bool=Get bIsPlayingEmote()
   ) -> (Pose) {
    alphaBoolBlend {blendOption=Linear}
    controlRigAssetReference {controlRigAsset=&/F_MED_QuaintToy_CtrlRig}
   }
   n9 = RigidBody!(ComponentPose=Local To Component(LocalPose=@n8), Alpha, OverrideWorldGravity:Vector="0,0,0", ExternalForce) -> (Pose) {
    bOverrideWorldGravity=true
    overridePhysicsAsset=&/F_MED_QuaintToy_SkirtPhysics
    simulationSpace=BaseBoneSpace
    baseBoneRef {boneName=pelvis}
    simSpaceSettings {
     dampingAlpha=0
     maxAngularAcceleration=15
     maxAngularVelocity=5
     maxLinearAcceleration=600
     maxLinearVelocity=200
     velocityScaleZ=0.6
     worldAlpha=1
     externalLinearDragV=<0.2 0.2 0.2>
    }
   }
   n10 = Anim Dynamics - Bone: dyn_skirt_05_bk_l!(ComponentPose=@n9, Alpha=1, GravityOverride:Vector="0,0,-980", ExternalForce="0,0,0") -> (Pose) {
    angularDampingOverride=1
    angularSpringConstant=2000
    bAngularSpring=true
    bOverrideAngularDamping=true
    bUseGravityOverride=true
    gravityScale=0.35
    simulationSpace=RootRelative
    alphaScaleBiasClamp {bInterpResult=true, interpSpeedDecreasing=5}
    boundBone {boneName=dyn_skirt_05_bk_l}
    componentLinearAccScale=<0.3 0.4 0.6>
    componentLinearVelScale=<0.3 0.4 0.6>
    gravityOverride {z=-980}
    physicsBodyDefinitions [{
      boundBone {boneName=dyn_skirt_05_bk_l}
      boxExtents=<20 30 30>
      constraintSetup {angularLimitsMax=<0 10 4>, angularLimitsMin=<0 -10 -6>, angularTarget=<1 0 0>}
      localJointOffset=<5 0 0>
     }]
    simSpaceSettings {maxAngularAcceleration=3000, maxAngularVelocity=3000, simSpaceAngularAlpha=0.1}
   }
   n11 = Anim Dynamics - Bone: dyn_skirt_04_bk_l!(ComponentPose=@n10, Alpha=1, GravityOverride:Vector="0,0,-980", ExternalForce="0,0,0") -> (Pose) {
    angularDampingOverride=1
    angularSpringConstant=2000
    bAngularSpring=true
    bOverrideAngularDamping=true
    bUseGravityOverride=true
    gravityScale=0.35
    simulationSpace=RootRelative
    alphaScaleBiasClamp {bInterpResult=true, interpSpeedDecreasing=5}
    boundBone {boneName=dyn_skirt_04_bk_l}
    componentLinearAccScale=<0.3 0.4 0.4>
    componentLinearVelScale=<0.3 0.4 0.4>
    gravityOverride {z=-980}
    physicsBodyDefinitions [{
      boundBone {boneName=dyn_skirt_04_bk_l}
      boxExtents=<20 30 30>
      constraintSetup {angularLimitsMax=<0 10 4>, angularLimitsMin=<0 -10 -25>, angularTarget=<1 0 0>}
      localJointOffset=<5 0 0>
     }]
    simSpaceSettings {maxAngularAcceleration=3000, maxAngularVelocity=3000, simSpaceAngularAlpha=0.1}
   }
   n12 = Anim Dynamics - Bone: dyn_skirt_05_bk!(ComponentPose=@n11, Alpha=1, GravityOverride:Vector="0,0,-980", ExternalForce="0,0,0") -> (Pose) {
    angularDampingOverride=1
    angularSpringConstant=2000
    bAngularSpring=true
    bOverrideAngularDamping=true
    bUseGravityOverride=true
    gravityScale=0.35
    simulationSpace=RootRelative
    alphaScaleBiasClamp {bInterpResult=true, interpSpeedDecreasing=5}
    boundBone {boneName=dyn_skirt_05_bk}
    componentLinearAccScale=<0.3 0.4 0.6>
    componentLinearVelScale=<0.3 0.4 0.6>
    gravityOverride {z=-980}
    physicsBodyDefinitions [{
      boundBone {boneName=dyn_skirt_05_bk}
      boxExtents=<20 30 30>
      constraintSetup {angularLimitsMax=<0 10 4>, angularLimitsMin=<0 -10 -6>, angularTarget=<1 0 0>}
      localJointOffset=<5 0 0>
     }]
    simSpaceSettings {maxAngularAcceleration=3000, maxAngularVelocity=3000, simSpaceAngularAlpha=0.1}
   }
   n13 = Anim Dynamics - Bone: dyn_skirt_04_bk!(ComponentPose=@n12, Alpha=1, GravityOverride:Vector="0,0,-980", ExternalForce="0,0,0") -> (Pose) {
    angularDampingOverride=1
    angularSpringConstant=2000
    bAngularSpring=true
    bOverrideAngularDamping=true
    bUseGravityOverride=true
    gravityScale=0.35
    simulationSpace=RootRelative
    alphaScaleBiasClamp {bInterpResult=true, interpSpeedDecreasing=5}
    boundBone {boneName=dyn_skirt_04_bk}
    componentLinearAccScale=<0.3 0.4 0.4>
    componentLinearVelScale=<0.3 0.4 0.4>
    gravityOverride {z=-980}
    physicsBodyDefinitions [{
      boundBone {boneName=dyn_skirt_04_bk}
      boxExtents=<20 30 30>
      constraintSetup {angularLimitsMax=<0 10 4>, angularLimitsMin=<0 -10 -35>, angularTarget=<1 0 0>}
      localJointOffset=<5 0 0>
     }]
    simSpaceSettings {maxAngularAcceleration=3000, maxAngularVelocity=3000, simSpaceAngularAlpha=0.1}
   }
   n14 = Anim Dynamics - Bone: dyn_skirt_05_bk_r!(ComponentPose=@n13, Alpha=1, GravityOverride:Vector="0,0,-980", ExternalForce="0,0,0") -> (Pose) {
    angularDampingOverride=1
    angularSpringConstant=2000
    bAngularSpring=true
    bOverrideAngularDamping=true
    bUseGravityOverride=true
    gravityScale=0.35
    simulationSpace=RootRelative
    alphaScaleBiasClamp {bInterpResult=true, interpSpeedDecreasing=5}
    boundBone {boneName=dyn_skirt_05_bk_r}
    componentLinearAccScale=<0.3 0.4 0.6>
    componentLinearVelScale=<0.3 0.4 0.6>
    gravityOverride {z=-980}
    physicsBodyDefinitions [{
      boundBone {boneName=dyn_skirt_05_bk_r}
      boxExtents=<20 30 30>
      constraintSetup {angularLimitsMax=<0 10 4>, angularLimitsMin=<0 -10 -6>, angularTarget=<1 0 0>}
      localJointOffset=<5 0 0>
     }]
    simSpaceSettings {maxAngularAcceleration=3000, maxAngularVelocity=3000, simSpaceAngularAlpha=0.1}
   }
   n15 = Anim Dynamics - Bone: dyn_skirt_04_bk_r!(ComponentPose=@n14, Alpha=1, GravityOverride:Vector="0,0,-980", ExternalForce="0,0,0") -> (Pose) {
    angularDampingOverride=1
    angularSpringConstant=2000
    bAngularSpring=true
    bOverrideAngularDamping=true
    bUseGravityOverride=true
    gravityScale=0.35
    simulationSpace=RootRelative
    alphaScaleBiasClamp {bInterpResult=true, interpSpeedDecreasing=5}
    boundBone {boneName=dyn_skirt_04_bk_r}
    componentLinearAccScale=<0.3 0.4 0.4>
    componentLinearVelScale=<0.3 0.4 0.4>
    gravityOverride {z=-980}
    physicsBodyDefinitions [{
      boundBone {boneName=dyn_skirt_04_bk_r}
      boxExtents=<20 30 30>
      constraintSetup {angularLimitsMax=<0 10 4>, angularLimitsMin=<0 -10 -25>, angularTarget=<1 0 0>}
      localJointOffset=<5 0 0>
     }]
    simSpaceSettings {maxAngularAcceleration=3000, maxAngularVelocity=3000, simSpaceAngularAlpha=0.1}
   }
   n16 = Anim Dynamics - Bone: dyn_rabbitHead(ComponentPose=@n15, Alpha=0.8, GravityOverride="0,0,-980", ExternalForce="0,0,0") {
    angularDampingOverride=0.999
    angularSpringConstant=3000
    bAngularSpring=true
    bOverrideAngularDamping=true
    bUseGravityOverride=true
    gravityScale=0.35
    simulationSpace=RootRelative
    boundBone {boneName=dyn_rabbitHead}
    componentAppliedLinearAccClamp=<3000 3000 3000>
    componentLinearAccScale=<0.3 0.3 0.3>
    componentLinearVelScale=<0.3 0.3 0.3>
    gravityOverride {z=-980}
    physicsBodyDefinitions [{
      boundBone {boneName=dyn_rabbitHead}
      boxExtents=<35 35 35>
      constraintSetup {angularLimitsMax=<0 10 0>, angularLimitsMin=<0 -10 0>, angularTarget=<1 0 0>}
      localJointOffset=<2 -1 0>
     }]
    relativeSpaceBone {boneName=spine_05}
    simSpaceSettings {simSpaceAngularAlpha=0.1}
   }
   n17 = RigidBody(ComponentPose=@n16, Alpha=0.8, OverrideWorldGravity="0,0,0") {
    bOverrideWorldGravity=true
    overridePhysicsAsset=&/F_MED_QuaintToy_RabbitPhysics
    simulationSpace=BaseBoneSpace
    baseBoneRef {boneName=dyn_skirt_03_fr_l}
    simSpaceSettings {
     dampingAlpha=0
     maxAngularAcceleration=15
     maxAngularVelocity=5
     maxLinearAcceleration=600
     maxLinearVelocity=200
     velocityScaleZ=0.6
     worldAlpha=1
     externalLinearDragV=<0.2 0.2 0.2>
    }
   }
   n18 = Anim Dynamics - Bone: dyn_rabbitArm_l(ComponentPose=@n17, Alpha=0.8, GravityOverride="0,0,-980", ExternalForce="0,0,0") {
    angularDampingOverride=0.999
    angularSpringConstant=2000
    bAngularSpring=true
    bOverrideAngularDamping=true
    bUseGravityOverride=true
    gravityScale=0.35
    simulationSpace=RootRelative
    boundBone {boneName=dyn_rabbitArm_l}
    componentAppliedLinearAccClamp=<8000 8000 8000>
    componentLinearAccScale=<0.4 0.4 0.5>
    componentLinearVelScale=<0.4 0.4 0.5>
    gravityOverride {z=-980}
    physicsBodyDefinitions [{
      boundBone {boneName=dyn_rabbitArm_l}
      boxExtents=<20 20 20>
      constraintSetup {angularLimitsMax=<0 12 20>, angularLimitsMin=<0 -3 -20>, angularTarget=<1 0 0>}
      localJointOffset=<3 1 0>
     }]
    relativeSpaceBone {boneName=spine_05}
    simSpaceSettings {maxAngularAcceleration=3000, maxAngularVelocity=3000, simSpaceAngularAlpha=0.1}
   }
   n19 = Anim Dynamics - Bone: dyn_rabbitArm_r(ComponentPose=@n18, Alpha=0.8, GravityOverride="0,0,-980", ExternalForce="0,0,0") {
    angularDampingOverride=0.999
    angularSpringConstant=2000
    bAngularSpring=true
    bOverrideAngularDamping=true
    bUseGravityOverride=true
    gravityScale=0.35
    simulationSpace=RootRelative
    boundBone {boneName=dyn_rabbitArm_r}
    componentAppliedLinearAccClamp=<8000 8000 8000>
    componentLinearAccScale=<0.4 0.4 0.5>
    componentLinearVelScale=<0.4 0.4 0.5>
    gravityOverride {z=-980}
    physicsBodyDefinitions [{
      boundBone {boneName=dyn_rabbitArm_r}
      boxExtents=<20 20 20>
      constraintSetup {angularLimitsMax=<0 5 0>, angularLimitsMin=<0 -12 -20>, angularTarget=<1 0 0>}
      localJointOffset=<3 -1 0>
     }]
    relativeSpaceBone {boneName=spine_05}
    simSpaceSettings {maxAngularAcceleration=3000, maxAngularVelocity=3000, simSpaceAngularAlpha=0.1}
   }
   n20 = Anim Dynamics - Bone: dyn_rabbitEar_01_r(ComponentPose=@n19, Alpha=0.8, GravityOverride="0,0,-980", ExternalForce="0,0,0") {
    angularDampingOverride=0.999
    angularSpringConstant=3000
    bAngularSpring=true
    bOverrideAngularDamping=true
    bUseGravityOverride=true
    gravityScale=0.35
    simulationSpace=RootRelative
    boundBone {boneName=dyn_rabbitEar_01_r}
    componentAppliedLinearAccClamp=<3000 3000 3000>
    componentLinearAccScale=<0.3 0.3 0.3>
    componentLinearVelScale=<0.3 0.3 0.3>
    gravityOverride {z=-980}
    physicsBodyDefinitions [{
      boundBone {boneName=dyn_rabbitEar_01_r}
      boxExtents=<20 20 20>
      constraintSetup {angularLimitsMax=<0 8 0>, angularLimitsMin=<0 -7 -20>, angularTarget=<1 0 0>}
      localJointOffset=<5 -1 0>
     }]
    relativeSpaceBone {boneName=spine_05}
    simSpaceSettings {simSpaceAngularAlpha=0.1}
   }
   n21 = Anim Dynamics - Bone: dyn_rabbitEar_01_l(ComponentPose=@n20, Alpha=0.8, GravityOverride="0,0,-980", ExternalForce="0,0,0") {
    angularDampingOverride=0.999
    angularSpringConstant=3000
    bAngularSpring=true
    bOverrideAngularDamping=true
    bUseGravityOverride=true
    gravityScale=0.35
    simulationSpace=RootRelative
    boundBone {boneName=dyn_rabbitEar_01_l}
    componentAppliedLinearAccClamp=<3000 3000 3000>
    componentLinearAccScale=<0.3 0.3 0.3>
    componentLinearVelScale=<0.3 0.3 0.3>
    gravityOverride {z=-980}
    physicsBodyDefinitions [{
      boundBone {boneName=dyn_rabbitEar_01_l}
      boxExtents=<20 20 30>
      constraintSetup {angularLimitsMax=<0 12 0>, angularLimitsMin=<0 -3 -20>, angularTarget=<1 0 0>}
      localJointOffset=<5 -1 0>
     }]
    relativeSpaceBone {boneName=spine_05}
    simSpaceSettings {simSpaceAngularAlpha=0.1}
   }
   n22 = Anim Dynamics - Bone: dyn_bobbyPin!(ComponentPose=@n21, Alpha=0.8, GravityOverride:Vector="0,0,-980", ExternalForce="0,0,0") -> (Pose) {
    angularDampingOverride=0.999
    angularSpringConstant=3000
    bAngularSpring=true
    bOverrideAngularDamping=true
    bUseGravityOverride=true
    gravityScale=0.35
    simulationSpace=RootRelative
    boundBone {boneName=dyn_bobbyPin}
    componentAppliedLinearAccClamp=<3000 3000 3000>
    componentLinearAccScale=<0.3 0.3 0.3>
    componentLinearVelScale=<0.3 0.3 0.3>
    gravityOverride {z=-980}
    physicsBodyDefinitions [{
      boundBone {boneName=dyn_bobbyPin}
      boxExtents=<28 28 28>
      constraintSetup {angularTargetAxis=AxisZ, twistAxis=AxisZ, angularLimitsMax=<0 7 0>, angularLimitsMin=<0 -7 0>, angularTarget=<1 0 0>}
      localJointOffset=<2 -1 0>
     }]
    relativeSpaceBone {boneName=spine_05}
    simSpaceSettings {simSpaceAngularAlpha=0.1}
   }
   n23 = Anim Dynamics - Bone: dyn_necklaceRibbon_l!(ComponentPose=@n22, Alpha=1, GravityOverride:Vector="0,0,-980", ExternalForce="0,0,0") -> (Pose) {
    angularDampingOverride=0.999
    angularSpringConstant=3000
    bAngularSpring=true
    bOverrideAngularDamping=true
    bUseGravityOverride=true
    gravityScale=0.35
    simulationSpace=RootRelative
    boundBone {boneName=dyn_necklaceRibbon_l}
    componentAppliedLinearAccClamp=<3000 3000 3000>
    componentLinearAccScale=<0.3 0.3 0.3>
    componentLinearVelScale=<0.3 0.3 0.3>
    gravityOverride {z=-980}
    physicsBodyDefinitions [{
      boundBone {boneName=dyn_necklaceRibbon_l}
      boxExtents=<16 16 16>
      constraintSetup {angularLimitsMax=<0 30 0>, angularLimitsMin=<0 -15 -40>, angularTarget=<1 0 0>}
      localJointOffset=<2 -1 0>
     }]
    relativeSpaceBone {boneName=spine_05}
    simSpaceSettings {simSpaceAngularAlpha=0.1}
   }
   n24 = Anim Dynamics - Bone: dyn_necklaceRibbon_r!(ComponentPose=@n23, Alpha=1, GravityOverride:Vector="0,0,-980", ExternalForce="0,0,0") -> (Pose) {
    angularDampingOverride=0.999
    angularSpringConstant=3000
    bAngularSpring=true
    bOverrideAngularDamping=true
    bUseGravityOverride=true
    gravityScale=0.35
    simulationSpace=RootRelative
    boundBone {boneName=dyn_necklaceRibbon_r}
    componentAppliedLinearAccClamp=<3000 3000 3000>
    componentLinearAccScale=<0.3 0.3 0.3>
    componentLinearVelScale=<0.3 0.3 0.3>
    gravityOverride {z=-980}
    physicsBodyDefinitions [{
      boundBone {boneName=dyn_necklaceRibbon_r}
      boxExtents=<16 16 16>
      constraintSetup {angularLimitsMax=<0 15 0>, angularLimitsMin=<0 -30 -40>, angularTarget=<1 0 0>}
      localJointOffset=<2 -1 0>
     }]
    relativeSpaceBone {boneName=spine_05}
    simSpaceSettings {simSpaceAngularAlpha=0.1}
   }
   n25 = Anim Dynamics - Bone: dyn_sleeveTop_l!(ComponentPose=@n24, Alpha=0.7, GravityOverride:Vector="0,0,-980", ExternalForce="0,0,0") -> (Pose) {
    alpha=0.7
    angularDampingOverride=1
    angularSpringConstant=10000
    bAngularSpring=true
    bOverrideAngularDamping=true
    bUseGravityOverride=true
    gravityScale=0.35
    simulationSpace=BoneRelative
    alphaScaleBiasClamp {bInterpResult=true, interpSpeedDecreasing=5}
    boundBone {boneName=dyn_sleeveTop_l}
    componentAppliedLinearAccClamp=<3000 3000 3000>
    componentLinearAccScale=<0.3 0.3 0.3>
    componentLinearVelScale=<0.3 0.4 3>
    gravityOverride {z=-980}
    physicsBodyDefinitions [{
      boundBone {boneName=dyn_sleeveTop_l}
      boxExtents=<32 32 32>
      constraintSetup {angularLimitsMax=<0 6 0>, angularLimitsMin=<0 -6 -60>, angularTarget=<1 0 0>}
      localJointOffset=<0.5 0 0>
     }]
    relativeSpaceBone {boneName=root}
    simSpaceSettings {simSpaceAngularAlpha=0.2}
   }
   n26 = Anim Dynamics - Bone: dyn_sleeveBottom_l!(ComponentPose=@n25, Alpha=0.7, GravityOverride:Vector="0,0,-980", ExternalForce="0,0,0") -> (Pose) {
    alpha=0.7
    angularDampingOverride=1
    angularSpringConstant=10000
    bAngularSpring=true
    bOverrideAngularDamping=true
    bUseGravityOverride=true
    gravityScale=0.35
    simulationSpace=BoneRelative
    alphaScaleBiasClamp {bInterpResult=true, interpSpeedDecreasing=5}
    boundBone {boneName=dyn_sleeveBottom_l}
    componentAppliedLinearAccClamp=<3000 3000 3000>
    componentLinearAccScale=<0.3 0.3 0.3>
    componentLinearVelScale=<0.3 0.4 3>
    gravityOverride {z=-980}
    physicsBodyDefinitions [{
      boundBone {boneName=dyn_sleeveBottom_l}
      boxExtents=<32 32 32>
      constraintSetup {angularLimitsMax=<0 6 0>, angularLimitsMin=<0 -6 -60>, angularTarget=<1 0 0>}
      localJointOffset=<0.5 0 0>
     }]
    relativeSpaceBone {boneName=root}
    simSpaceSettings {simSpaceAngularAlpha=0.2}
   }
   n27 = Anim Dynamics - Bone: dyn_sleeveFront_l!(ComponentPose=@n26, Alpha=0.7, GravityOverride:Vector="0,0,-980", ExternalForce="0,0,0") -> (Pose) {
    alpha=0.7
    angularDampingOverride=1
    angularSpringConstant=10000
    bAngularSpring=true
    bOverrideAngularDamping=true
    bUseGravityOverride=true
    gravityScale=0.35
    simulationSpace=BoneRelative
    alphaScaleBiasClamp {bInterpResult=true, interpSpeedDecreasing=5}
    boundBone {boneName=dyn_sleeveFront_l}
    componentAppliedLinearAccClamp=<3000 3000 3000>
    componentLinearAccScale=<0.3 0.3 0.3>
    componentLinearVelScale=<0.3 0.4 3>
    gravityOverride {z=-980}
    physicsBodyDefinitions [{
      boundBone {boneName=dyn_sleeveFront_l}
      boxExtents=<32 32 32>
      constraintSetup {angularLimitsMax=<0 6 0>, angularLimitsMin=<0 -6 -60>, angularTarget=<1 0 0>}
      localJointOffset=<0.5 0 0>
     }]
    relativeSpaceBone {boneName=root}
    simSpaceSettings {simSpaceAngularAlpha=0.2}
   }
   n28 = Anim Dynamics - Bone: dyn_sleeveBack_l!(ComponentPose=@n27, Alpha=0.7, GravityOverride:Vector="0,0,-980", ExternalForce="0,0,0") -> (Pose) {
    alpha=0.7
    angularDampingOverride=1
    angularSpringConstant=10000
    bAngularSpring=true
    bOverrideAngularDamping=true
    bUseGravityOverride=true
    gravityScale=0.35
    simulationSpace=BoneRelative
    alphaScaleBiasClamp {bInterpResult=true, interpSpeedDecreasing=5}
    boundBone {boneName=dyn_sleeveBack_l}
    componentAppliedLinearAccClamp=<3000 3000 3000>
    componentLinearAccScale=<0.3 0.3 0.3>
    componentLinearVelScale=<0.3 0.4 3>
    gravityOverride {z=-980}
    physicsBodyDefinitions [{
      boundBone {boneName=dyn_sleeveBack_l}
      boxExtents=<32 32 32>
      constraintSetup {angularLimitsMax=<0 6 0>, angularLimitsMin=<0 -6 -60>, angularTarget=<1 0 0>}
      localJointOffset=<0.5 0 0>
     }]
    relativeSpaceBone {boneName=root}
    simSpaceSettings {simSpaceAngularAlpha=0.2}
   }
   n29 = Anim Dynamics - Bone: dyn_sleeveTop_r!(ComponentPose=@n28, Alpha=0.7, GravityOverride:Vector="0,0,-980", ExternalForce="0,0,0") -> (Pose) {
    alpha=0.7
    angularDampingOverride=1
    angularSpringConstant=10000
    bAngularSpring=true
    bOverrideAngularDamping=true
    bUseGravityOverride=true
    gravityScale=0.35
    simulationSpace=BoneRelative
    alphaScaleBiasClamp {bInterpResult=true, interpSpeedDecreasing=5}
    boundBone {boneName=dyn_sleeveTop_r}
    componentAppliedLinearAccClamp=<3000 3000 3000>
    componentLinearAccScale=<0.3 0.3 0.3>
    componentLinearVelScale=<0.3 0.4 3>
    gravityOverride {z=-980}
    physicsBodyDefinitions [{
      boundBone {boneName=dyn_sleeveTop_r}
      boxExtents=<32 32 32>
      constraintSetup {angularLimitsMax=<0 6 0>, angularLimitsMin=<0 -6 -60>, angularTarget=<1 0 0>}
      localJointOffset=<0.5 0 0>
     }]
    relativeSpaceBone {boneName=root}
    simSpaceSettings {simSpaceAngularAlpha=0.2}
   }
   n30 = Anim Dynamics - Bone: dyn_sleeveBottom_r!(ComponentPose=@n29, Alpha=0.7, GravityOverride:Vector="0,0,-980", ExternalForce="0,0,0") -> (Pose) {
    alpha=0.7
    angularDampingOverride=1
    angularSpringConstant=10000
    bAngularSpring=true
    bOverrideAngularDamping=true
    bUseGravityOverride=true
    gravityScale=0.35
    simulationSpace=BoneRelative
    alphaScaleBiasClamp {bInterpResult=true, interpSpeedDecreasing=5}
    boundBone {boneName=dyn_sleeveBottom_r}
    componentAppliedLinearAccClamp=<3000 3000 3000>
    componentLinearAccScale=<0.3 0.3 0.3>
    componentLinearVelScale=<0.3 0.4 3>
    gravityOverride {z=-980}
    physicsBodyDefinitions [{
      boundBone {boneName=dyn_sleeveBottom_r}
      boxExtents=<32 32 32>
      constraintSetup {angularLimitsMax=<0 6 0>, angularLimitsMin=<0 -6 -60>, angularTarget=<1 0 0>}
      localJointOffset=<0.5 0 0>
     }]
    relativeSpaceBone {boneName=root}
    simSpaceSettings {simSpaceAngularAlpha=0.2}
   }
   n31 = Anim Dynamics - Bone: dyn_sleeveFront_r!(ComponentPose=@n30, Alpha=0.7, GravityOverride:Vector="0,0,-980", ExternalForce="0,0,0") -> (Pose) {
    alpha=0.7
    angularDampingOverride=1
    angularSpringConstant=10000
    bAngularSpring=true
    bOverrideAngularDamping=true
    bUseGravityOverride=true
    gravityScale=0.35
    simulationSpace=BoneRelative
    alphaScaleBiasClamp {bInterpResult=true, interpSpeedDecreasing=5}
    boundBone {boneName=dyn_sleeveFront_r}
    componentAppliedLinearAccClamp=<3000 3000 3000>
    componentLinearAccScale=<0.3 0.3 0.3>
    componentLinearVelScale=<0.3 0.4 3>
    gravityOverride {z=-980}
    physicsBodyDefinitions [{
      boundBone {boneName=dyn_sleeveFront_r}
      boxExtents=<32 32 32>
      constraintSetup {angularLimitsMax=<0 6 0>, angularLimitsMin=<0 -6 -60>, angularTarget=<1 0 0>}
      localJointOffset=<0.5 0 0>
     }]
    relativeSpaceBone {boneName=root}
    simSpaceSettings {simSpaceAngularAlpha=0.2}
   }
   n32 = Anim Dynamics - Bone: dyn_sleeveBack_r!(ComponentPose=@n31, Alpha=0.7, GravityOverride:Vector="0,0,-980", ExternalForce="0,0,0") -> (Pose) {
    alpha=0.7
    angularDampingOverride=1
    angularSpringConstant=10000
    bAngularSpring=true
    bOverrideAngularDamping=true
    bUseGravityOverride=true
    gravityScale=0.35
    simulationSpace=BoneRelative
    alphaScaleBiasClamp {bInterpResult=true, interpSpeedDecreasing=5}
    boundBone {boneName=dyn_sleeveBack_r}
    componentAppliedLinearAccClamp=<3000 3000 3000>
    componentLinearAccScale=<0.3 0.3 0.3>
    componentLinearVelScale=<0.3 0.4 3>
    gravityOverride {z=-980}
    physicsBodyDefinitions [{
      boundBone {boneName=dyn_sleeveBack_r}
      boxExtents=<32 32 32>
      constraintSetup {angularLimitsMax=<0 6 0>, angularLimitsMin=<0 -6 -60>, angularTarget=<1 0 0>}
      localJointOffset=<0.5 0 0>
     }]
    relativeSpaceBone {boneName=root}
    simSpaceSettings {simSpaceAngularAlpha=0.2}
   }
   n33 = Anim Dynamics - Bone: dyn_ribbon_01_r!(ComponentPose=@n32, Alpha=1, GravityOverride:Vector="0,0,-980", ExternalForce="0,0,0") -> (Pose) {
    angularDampingOverride=0.999
    angularSpringConstant=4000
    bAngularSpring=true
    bOverrideAngularDamping=true
    bUseGravityOverride=true
    gravityScale=0.35
    simulationSpace=RootRelative
    boundBone {boneName=dyn_ribbon_01_r}
    componentAppliedLinearAccClamp=<3000 3000 3000>
    componentLinearAccScale=<0.4 0.4 0.5>
    componentLinearVelScale=<0.4 0.4 0.5>
    gravityOverride {z=-980}
    physicsBodyDefinitions [{
      boundBone {boneName=dyn_ribbon_01_r}
      boxExtents=<28 28 28>
      constraintSetup {angularLimitsMax=<0 40 1>, angularLimitsMin=<0 -25 -40>, angularTarget=<1 0 0>}
      localJointOffset=<2 -1 0>
     }]
    relativeSpaceBone {boneName=spine_05}
    simSpaceSettings {simSpaceAngularAlpha=0.3}
   }
   n34 = Anim Dynamics - Bone: dyn_ribbon_02_r!(ComponentPose=@n33, Alpha=1, GravityOverride:Vector="0,0,-980", ExternalForce="0,0,0") -> (Pose) {
    angularDampingOverride=0.999
    angularSpringConstant=4000
    bAngularSpring=true
    bOverrideAngularDamping=true
    bUseGravityOverride=true
    gravityScale=0.35
    simulationSpace=RootRelative
    boundBone {boneName=dyn_ribbon_02_r}
    componentAppliedLinearAccClamp=<3000 3000 3000>
    componentLinearAccScale=<0.4 0.4 0.5>
    componentLinearVelScale=<0.4 0.4 0.5>
    gravityOverride {z=-980}
    physicsBodyDefinitions [{
      boundBone {boneName=dyn_ribbon_02_r}
      boxExtents=<28 28 28>
      constraintSetup {angularLimitsMax=<0 40 1>, angularLimitsMin=<0 -25 -40>, angularTarget=<1 0 0>}
      localJointOffset=<2 -1 0>
     }]
    relativeSpaceBone {boneName=spine_05}
    simSpaceSettings {simSpaceAngularAlpha=0.3}
   }
   n35 = Anim Dynamics - Bone: dyn_ribbon_03_r!(ComponentPose=@n34, Alpha=1, GravityOverride:Vector="0,0,-980", ExternalForce="0,0,0") -> (Pose) {
    angularDampingOverride=0.999
    angularSpringConstant=4000
    bAngularSpring=true
    bOverrideAngularDamping=true
    bUseGravityOverride=true
    gravityScale=0.35
    simulationSpace=RootRelative
    boundBone {boneName=dyn_ribbon_03_r}
    componentAppliedLinearAccClamp=<3000 3000 3000>
    componentLinearAccScale=<0.4 0.4 0.5>
    componentLinearVelScale=<0.4 0.4 0.5>
    gravityOverride {z=-980}
    physicsBodyDefinitions [{
      boundBone {boneName=dyn_ribbon_03_r}
      boxExtents=<28 28 28>
      constraintSetup {angularLimitsMax=<0 40 1>, angularLimitsMin=<0 -25 -40>, angularTarget=<1 0 0>}
      localJointOffset=<2 -1 0>
     }]
    relativeSpaceBone {boneName=spine_05}
    simSpaceSettings {simSpaceAngularAlpha=0.3}
   }
   n36 = Anim Dynamics - Bone: dyn_bow_r!(ComponentPose=@n35, Alpha=0.5, GravityOverride:Vector="0,0,-980", ExternalForce="0,0,0") -> (Pose) {
    angularDampingOverride=0.999
    angularSpringConstant=7000
    bAngularSpring=true
    bOverrideAngularDamping=true
    bUseGravityOverride=true
    gravityScale=0.35
    simulationSpace=RootRelative
    boundBone {boneName=dyn_bow_r}
    componentAppliedLinearAccClamp=<3000 3000 3000>
    componentLinearAccScale=<0.3 0.3 0.3>
    componentLinearVelScale=<0.3 0.3 0.3>
    gravityOverride {z=-980}
    physicsBodyDefinitions [{
      boundBone {boneName=dyn_bow_r}
      boxExtents=<28 28 28>
      constraintSetup {angularLimitsMax=<0 40 3>, angularLimitsMin=<0 -25 -40>, angularTarget=<1 0 0>}
      localJointOffset=<1.5 0 -2>
     }]
    relativeSpaceBone {boneName=spine_05}
    simSpaceSettings {simSpaceAngularAlpha=0.1}
   }
   n37 = Anim Dynamics - Bone: dyn_ribbon_01_l!(ComponentPose=@n36, Alpha=1, GravityOverride:Vector="0,0,-980", ExternalForce="0,0,0") -> (Pose) {
    angularDampingOverride=0.999
    angularSpringConstant=4000
    bAngularSpring=true
    bOverrideAngularDamping=true
    bUseGravityOverride=true
    gravityScale=0.35
    simulationSpace=RootRelative
    boundBone {boneName=dyn_ribbon_01_l}
    componentAppliedLinearAccClamp=<3000 3000 3000>
    componentLinearAccScale=<0.4 0.4 0.5>
    componentLinearVelScale=<0.4 0.4 0.5>
    gravityOverride {z=-980}
    physicsBodyDefinitions [{
      boundBone {boneName=dyn_ribbon_01_l}
      boxExtents=<28 28 28>
      constraintSetup {angularLimitsMax=<0 40 1>, angularLimitsMin=<0 -25 -40>, angularTarget=<1 0 0>}
      localJointOffset=<2 -1 0>
     }]
    relativeSpaceBone {boneName=spine_05}
    simSpaceSettings {simSpaceAngularAlpha=0.3}
   }
   n38 = Anim Dynamics - Bone: dyn_ribbon_02_l!(ComponentPose=@n37, Alpha=1, GravityOverride:Vector="0,0,-980", ExternalForce="0,0,0") -> (Pose) {
    angularDampingOverride=0.999
    angularSpringConstant=4000
    bAngularSpring=true
    bOverrideAngularDamping=true
    bUseGravityOverride=true
    gravityScale=0.35
    simulationSpace=RootRelative
    boundBone {boneName=dyn_ribbon_02_l}
    componentAppliedLinearAccClamp=<3000 3000 3000>
    componentLinearAccScale=<0.4 0.4 0.5>
    componentLinearVelScale=<0.4 0.4 0.5>
    gravityOverride {z=-980}
    physicsBodyDefinitions [{
      boundBone {boneName=dyn_ribbon_02_l}
      boxExtents=<28 28 28>
      constraintSetup {angularLimitsMax=<0 40 1>, angularLimitsMin=<0 -25 -40>, angularTarget=<1 0 0>}
      localJointOffset=<2 -1 0>
     }]
    relativeSpaceBone {boneName=spine_05}
    simSpaceSettings {simSpaceAngularAlpha=0.3}
   }
   n39 = Anim Dynamics - Bone: dyn_ribbon_03_l!(ComponentPose=@n38, Alpha=1, GravityOverride:Vector="0,0,-980", ExternalForce="0,0,0") -> (Pose) {
    angularDampingOverride=0.999
    angularSpringConstant=4000
    bAngularSpring=true
    bOverrideAngularDamping=true
    bUseGravityOverride=true
    gravityScale=0.35
    simulationSpace=RootRelative
    boundBone {boneName=dyn_ribbon_03_l}
    componentAppliedLinearAccClamp=<3000 3000 3000>
    componentLinearAccScale=<0.4 0.4 0.5>
    componentLinearVelScale=<0.4 0.4 0.5>
    gravityOverride {z=-980}
    physicsBodyDefinitions [{
      boundBone {boneName=dyn_ribbon_03_l}
      boxExtents=<28 28 28>
      constraintSetup {angularLimitsMax=<0 40 1>, angularLimitsMin=<0 -25 -40>, angularTarget=<1 0 0>}
      localJointOffset=<2 -1 0>
     }]
    relativeSpaceBone {boneName=spine_05}
    simSpaceSettings {simSpaceAngularAlpha=0.3}
   }
   n40 = Anim Dynamics - Bone: dyn_bow_l!(ComponentPose=@n39, Alpha=0.5, GravityOverride:Vector="0,0,-980", ExternalForce="0,0,0") -> (Pose) {
    angularDampingOverride=0.999
    angularSpringConstant=7000
    bAngularSpring=true
    bOverrideAngularDamping=true
    bUseGravityOverride=true
    gravityScale=0.35
    simulationSpace=RootRelative
    boundBone {boneName=dyn_bow_l}
    componentAppliedLinearAccClamp=<3000 3000 3000>
    componentLinearAccScale=<0.3 0.3 0.3>
    componentLinearVelScale=<0.3 0.3 0.3>
    gravityOverride {z=-980}
    physicsBodyDefinitions [{
      boundBone {boneName=dyn_bow_l}
      boxExtents=<28 28 28>
      constraintSetup {angularLimitsMax=<0 40 3>, angularLimitsMin=<0 -25 -40>, angularTarget=<1 0 0>}
      localJointOffset=<1.5 0 2>
     }]
    relativeSpaceBone {boneName=spine_05}
    simSpaceSettings {simSpaceAngularAlpha=0.1}
   }
   n41 = Get bIsSkyDiving()
   n42 = "Transform (Modify) Bone - Bone: base_ribbon_r"!(ComponentPose=@n40, bAlphaBoolEnabled:bool=@n41, Translation="0.041,0.716,0.857", Rotation="0,0,0", Scale="1,1,1") -> (Pose) {
    alphaInputType=Bool
    rotationMode=BMM_Additive
    rotationSpace=BCS_BoneSpace
    translationMode=BMM_Additive
    alphaBoolBlend {blendInTime=0.6, blendOutTime=0.6}
    boneToModify {boneName=base_ribbon_r}
    translation=<0.041 0.716 0.857>
   }
   n43 = "Transform (Modify) Bone - Bone: base_ribbon_l"!(ComponentPose=@n42, bAlphaBoolEnabled:bool=@n41, Translation="0.033,0.564,0.821", Rotation="0,0,0", Scale="1,1,1") -> (Pose) {
    alphaInputType=Bool
    rotationMode=BMM_Additive
    rotationSpace=BCS_BoneSpace
    translationMode=BMM_Additive
    alphaBoolBlend {blendInTime=0.6, blendOutTime=0.6}
    boneToModify {boneName=base_ribbon_l}
    translation=<0.033 0.564 0.821>
   }
   n44 = "Transform (Modify) Bone - Bone: dyn_rabbitArm_l"(ComponentPose=@n43, Alpha=1, Translation="0,0,0", Rotation="0.075,-9.662,0.028", Scale="1,1,1") {rotationMode=BMM_Additive, rotationSpace=BCS_BoneSpace, boneToModify {boneName=dyn_rabbitArm_l}, rotation=<rot 0.075 -9.662 0.028>}
   n45 = "Transform (Modify) Bone - Bone: base_rabbitEar_l"(ComponentPose=@n44, Alpha=1, Translation="0,0,0", Rotation="-1.384,-8.301,-0.007", Scale="1,1,1") {rotationMode=BMM_Additive, rotationSpace=BCS_BoneSpace, boneToModify {boneName=base_rabbitEar_l}, rotation=<rot -1.384 -8.301 -0.007>}
   n46 = Get walkingNotCrouching() -> (walkingNotCrouching:bool)
   n47 = "Transform (Modify) Bone - Bone: base_skirt_fr_l"!(ComponentPose=@n45, bAlphaBoolEnabled:bool=@n46, Translation="0,0,0", Rotation="0,0,-3.4", Scale="1,1,1") -> (Pose) {
    alphaInputType=Bool
    rotationMode=BMM_Additive
    alphaBoolBlend {blendInTime=0.5, blendOutTime=0.5}
    boneToModify {boneName=base_skirt_fr_l}
    rotation {roll=-3.4}
   }
   n48 = "Transform (Modify) Bone - Bone: base_skirt_fr_r"!(ComponentPose=@n47, bAlphaBoolEnabled:bool=@n46, Translation="0,0,0", Rotation="0.000,0.000,-4.491", Scale="1,1,1") -> (Pose) {
    alphaInputType=Bool
    rotationMode=BMM_Additive
    alphaBoolBlend {blendInTime=0.5, blendOutTime=0.5}
    boneToModify {boneName=base_skirt_fr_r}
    rotation {roll=-4.491}
   }
   n49 = "Transform (Modify) Bone - Bone: base_skirt_fr"!(ComponentPose=@n48, bAlphaBoolEnabled:bool=@n46, Translation="0,0,0", Rotation="0.000,0.000,-6.003", Scale="1,1,1") -> (Pose) {
    alphaInputType=Bool
    rotationMode=BMM_Additive
    alphaBoolBlend {blendInTime=0.5, blendOutTime=0.5}
    boneToModify {boneName=base_skirt_fr}
    rotation {roll=-6.003}
   }
   n50 = Get bIsTacticalSprinting()
   n51 = "Transform (Modify) Bone - Bone: base_bobbyPin"!(ComponentPose=@n49, bAlphaBoolEnabled:bool=@n50, Translation="0,0,0", Rotation="1.566,10.798,-8.164", Scale="1,1,1") -> (Pose) {
    alphaInputType=Bool
    rotationMode=BMM_Additive
    translationMode=BMM_Additive
    alphaBoolBlend {blendInTime=0.5, blendOutTime=0.5}
    boneToModify {boneName=base_bobbyPin}
    rotation=<rot 1.566 10.798 -8.164>
   }
   n52 = "Transform (Modify) Bone - Bone: base_rabbitHead"!(ComponentPose=@n51, bAlphaBoolEnabled:bool=@n50, Translation="0,0.79,0", Rotation="0,0,0", Scale="1,1,1") -> (Pose) {
    alphaInputType=Bool
    rotationMode=BMM_Additive
    translationMode=BMM_Additive
    alphaBoolBlend {blendInTime=0.5, blendOutTime=0.5}
    boneToModify {boneName=base_rabbitHead}
    translation {y=0.79}
   }
   Output Pose(Result=Component To Local(ComponentPose=@n52)) {layerGroup=None}
   Get bIsSkyDiving()
  ]}
 EventGraph {nodes [
   Event BlueprintUpdateAnimation()
   > n1 = Sequence()
   > n2 = "Gravity Override State Function (Parameter Asset) (Nativized)"(
    CharacterPartAnimInstance=Self-Reference()
    GameTimeInSeconds=GetGameTimeInSeconds()
    DebugJointName=spine_05
    Parameters=&/F_MED_QuaintToy_Parameters
   )
   > Set Gravity Override(Gravity Override=@n2)
   n3 = CheckForBackbling()
   > n4 = Branch(Condition=OR Boolean!(A=@n3, B=@n3.bIsBackblingACape, C:bool=@n3.bIsBackblingLong) -> (ReturnValue))
   > Set backblingAlpha(backblingAlpha:double=0) -> (Output_Get:double)
   Set backblingAlpha(execute=@n4.else, backblingAlpha=1)
   n5 = Get bIsCrouching()
   Set walkingNotCrouching(execute=@n1.then_1, walkingNotCrouching:bool=AND Boolean(A=Get bIsMoving2D(), B=NOT Boolean(A=@n5))) -> (Output_Get:bool)
  ]}
]
variables [
 backblingAlpha {type=number}
 Gravity Override {type=Vector}
 walkingNotCrouching {type=boolean}
]
