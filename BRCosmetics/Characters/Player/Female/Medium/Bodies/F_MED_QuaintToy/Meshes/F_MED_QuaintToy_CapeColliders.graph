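# Physics asset of collision shapes for the cape on the F_MED_QuaintToy female medium player body, as its CapeColliders name indicates.
# Capsules cover the pelvis, spine, clavicles, head and feet, with some spine bones doubled up.
# Capsules on the center, left and right back-skirt bone chains add the skirt to the collision volume.
# Arm, leg and neck_02 bodies hold no shapes, and no constraints link the bodies.
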PhysicsAsset /BRCosmetics/Characters/Player/Female/Medium/Bodies/F_MED_QuaintToy/Meshes/F_MED_QuaintToy_CapeColliders
bodies {
 calf_l []
 calf_r []
 clavicle_l [{shapeType=Capsule, length=6.5623, radius=7.5194, center=<3.9566 -1.0862 -2.1127>, rotation=<rot -77.5566 177.3915 7.7314>}]
 clavicle_r [{shapeType=Capsule, length=6.5623, radius=7.5194, center=<-3.9566 1.0862 2.1127>, rotation=<rot -77.5566 177.3915 -172.2686>}]
 dyn_skirt_02_bk [{
   shapeType=Capsule
   length=13.5466
   radius=7.8851
   shapeName=dyn_skirt_05_bk_capsule
   center=<5.683 0.4722 0.6635>
   rotation=<rot -1.7538 -89.6362 -1.3176>
  }]
 dyn_skirt_02_bk_l [{
   shapeType=Capsule
   length=16.904
   radius=7.8851
   shapeName=dyn_skirt_05_bk_capsule
   center=<5.6753 1.836 0.6635>
   rotation=<rot 12.8107 -89.9998 -1.6736>
  }]
 dyn_skirt_02_bk_r [{
   shapeType=Capsule
   length=16.904
   radius=7.8851
   shapeName=dyn_skirt_05_bk_capsule
   center=<5.6753 1.8361 0.6635>
   rotation=<rot 12.8108 -89.9998 -1.6735>
  }]
 dyn_skirt_03_bk [{
   shapeType=Capsule
   length=13.5466
   radius=7.8851
   shapeName=dyn_skirt_05_bk_capsule
   center=<5.6998 2.4086 -0.3239>
   rotation=<rot -3.7018 -89.5228 -1.6353>
  }]
 dyn_skirt_03_bk_l [{
   shapeType=Capsule
   length=16.5598
   radius=7.8851
   shapeName=dyn_skirt_05_bk_capsule
   center=<12.2765 0.5954 1.1434>
   rotation=<rot 6.017 -89.8006 1.8278>
  }]
 dyn_skirt_03_bk_r [{
   shapeType=Capsule
   length=16.5598
   radius=7.8851
   shapeName=dyn_skirt_05_bk_capsule
   center=<5.6753 2.4756 0.5181>
   rotation=<rot 12.8108 -89.9998 -1.6735>
  }]
 dyn_skirt_04_bk [{
   shapeType=Capsule
   length=13.5466
   radius=7.8851
   shapeName=dyn_skirt_05_bk_capsule
   center=<5.6928 3.3783 -0.2985>
   rotation=<rot -4.785 -89.4917 -1.6376>
  }]
 dyn_skirt_04_bk_l [{
   shapeType=Capsule
   length=13.5466
   radius=7.8851
   shapeName=dyn_skirt_05_bk_capsule
   center=<5.6753 5.3434 -0.134>
   rotation=<rot 12.8107 -89.9998 -1.6736>
  }]
 dyn_skirt_04_bk_r [{
   shapeType=Capsule
   length=13.5466
   radius=7.8851
   shapeName=dyn_skirt_05_bk_capsule
   center=<5.6753 3.4204 0.3032>
   rotation=<rot 12.8108 -89.9998 -1.6735>
  }]
 dyn_skirt_05_bk [{
   shapeType=Capsule
   length=13.5466
   radius=7.8851
   shapeName=dyn_skirt_05_bk_capsule
   center=<5.6753 5.0806 -0.0743>
   rotation=<rot -3.6082 -89.5254 -1.6351>
  }]
 dyn_skirt_05_bk_l [{
   shapeType=Capsule
   length=16.9573
   radius=7.8851
   shapeName=dyn_skirt_05_bk_capsule
   center=<5.6442 5.9672 2.6105>
   rotation=<rot 12.8028 -90.2723 -0.6949>
  }]
 dyn_skirt_05_bk_r [{
   shapeType=Capsule
   length=15.8586
   radius=7.8851
   shapeName=dyn_skirt_05_bk_capsule
   center=<5.7512 5.1632 -0.9252>
   rotation=<rot 12.8029 -90.2722 -5.3987>
  }]
 foot_l [{shapeType=Capsule, length=11.0765, radius=8.0344, center=<-4.1811 6.0027 -0.338>, rotation=<rot 0.0388 1.4701 -89.3982>}]
 foot_r [{shapeType=Capsule, length=11.0765, radius=8.0344, center=<4.1811 -6.0027 0.338>, rotation=<rot 0.0388 1.4701 90.6018>}]
 head [{shapeType=Capsule, length=0, radius=18, center=<9.0697 0.0746 0>, rotation=<rot 0 -90 0>}]
 lowerarm_l []
 lowerarm_r []
 neck_02 []
 pelvis [{shapeType=Capsule, length=9.6042, radius=9.2282, center=<-2.4471 -2.5289 -0.1981>}]
 spine_01 [{shapeType=Capsule, length=4.3888, radius=11.3119, center=<-5.3927 -7.7613 0>}]
 spine_02 [
  {shapeType=Capsule, length=2.5219, radius=9.3767, center=<-2.6885 -10.6951 0>}
  {shapeType=Capsule, length=20.6875, radius=9.3125, center=<-1.8317 1.9362 0>, rotation=<rot 90 0 0>}
 ]
 spine_03 [
  {shapeType=Capsule, length=15.0369, radius=8.4854, center=<0.0574 -7.1113 0>}
  {shapeType=Capsule, length=20.7188, radius=9.2812, center=<-1.2163 0.549 0>, rotation=<rot 90 180 -175.0732>}
 ]
 spine_04 [
  {shapeType=Capsule, length=9.0076, radius=10.3916, center=<4.3068 -2.988 0>}
  {shapeType=Capsule, length=10.0312, radius=9.2812, center=<-1.2326 1.308 0>, rotation=<rot 90 0 0>}
 ]
 spine_05 [{shapeType=Capsule, length=9.678, radius=12.3571, center=<0.6969 1.91 0>}]
 thigh_l []
 thigh_r []
 upperarm_l []
 upperarm_r []
}
constraints []
