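# Physics asset for the F_MED_QuaintToy female character's face accessory part, setting up two long dynamic hair strands, left and right.
# Each side chains eight capsule bodies from a small base hair body, with twist-locked constraints and narrow swing limits.
# Unconstrained capsules on head, neck, spine, pelvis and upper arms give the hair body shapes to collide with, including an oversized extra capsule on spine_05.
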PhysicsAsset /BRCosmetics/Characters/Player/Female/Medium/Bodies/F_MED_QuaintToy/Meshes/Parts/F_MED_QuaintToy_FaceAcc_Physics
bodies {
 base_hair_l [{shapeType=Capsule, length=0.505, radius=0.505, shapeName=base_hair_l_capsule, rotation=<rot 0 -90 0>}]
 base_hair_r [{shapeType=Capsule, length=0.505, radius=0.505, shapeName=base_hair_r_capsule, rotation=<rot 0 -90 0>}]
 dyn_hair_01_l [{shapeType=Capsule, length=0, radius=4.4874, shapeName=dyn_hair_03_l_capsule, center=<4.6799 -1.0212 0.0092>}]
 dyn_hair_01_r [{shapeType=Capsule, length=0, radius=4.4874, shapeName=dyn_hair_03_r_capsule, center=<4.2353 1.0212 -0.0092>, rotation=<rot 0 0 -180>}]
 dyn_hair_02_l [{shapeType=Capsule, length=0.0598, radius=5.656, shapeName=dyn_hair_03_l_capsule, center=<6.3242 0.5874 -1.3088>}]
 dyn_hair_02_r [{shapeType=Capsule, length=0.0598, radius=5.656, shapeName=dyn_hair_03_r_capsule, center=<5.7557 0.7237 1.3088>, rotation=<rot 0 0 -180>}]
 dyn_hair_03_l [{shapeType=Capsule, length=0, radius=5.656, shapeName=dyn_hair_03_l_capsule, center=<3.7028 -0.768 0>}]
 dyn_hair_03_r [{shapeType=Capsule, length=0.8936, radius=5.656, shapeName=dyn_hair_03_r_capsule, center=<4.0348 -0.3995 0>, rotation=<rot 0 0 -180>}]
 dyn_hair_04_l [{shapeType=Capsule, length=1.821, radius=5.656, shapeName=dyn_hair_03_l_capsule, center=<3.7028 -2.2559 -1.8532>}]
 dyn_hair_04_r [{shapeType=Capsule, length=1.821, radius=5.656, shapeName=dyn_hair_03_r_capsule, center=<4.0348 -2.1634 1.8532>, rotation=<rot 0 0 -180>}]
 dyn_hair_05_l [{shapeType=Capsule, length=1.9259, radius=5.656, shapeName=dyn_hair_03_l_capsule, center=<3.7028 -1.74 -0.5083>}]
 dyn_hair_05_r [{shapeType=Capsule, length=1.9259, radius=5.656, shapeName=dyn_hair_03_r_capsule, center=<4.0348 -1.9126 0.5083>, rotation=<rot 0 0 -180>}]
 dyn_hair_06_l [{shapeType=Capsule, length=0.9634, radius=5.656, shapeName=dyn_hair_03_l_capsule, center=<3.7028 -0.8604 0>}]
 dyn_hair_06_r [{shapeType=Capsule, length=2.2182, radius=5.656, shapeName=dyn_hair_03_r_capsule, center=<5.8784 -0.8948 1.2126>, rotation=<rot 0 0 -180>}]
 dyn_hair_07_l [{shapeType=Capsule, length=0, radius=5.656, shapeName=dyn_hair_03_l_capsule, center=<3.7028 -0.2431 0>}]
 dyn_hair_07_r [{shapeType=Capsule, length=0, radius=5.656, shapeName=dyn_hair_03_r_capsule, center=<3.6306 -0.219 0>, rotation=<rot 0 0 -180>}]
 dyn_hair_08_l [{shapeType=Capsule, length=0.2493, radius=3.3872, shapeName=dyn_hair_03_l_capsule, center=<3.8011 0.8289 0.1235>}]
 dyn_hair_08_r [{shapeType=Capsule, length=0.9052, radius=4.4763, shapeName=dyn_hair_03_r_capsule, center=<5.2957 0.7191 0.6669>, rotation=<rot 0 0 -180>}]
 head [{shapeType=Capsule, length=0.379, radius=8.0013, shapeName=head_capsule, center=<6.3326 -2.1199 -0.8356>, rotation=<rot 7.2959 -90 2.5625>}]
 neck_01 [{shapeType=Capsule, length=0, radius=4.2813, shapeName=neck_01_capsule, rotation=<rot 0 -90 0>}]
 neck_02 [{shapeType=Capsule, length=0, radius=4.2813, shapeName=neck_02_capsule, rotation=<rot 0 -90 0>}]
 pelvis [{shapeType=Capsule, length=16.5067, radius=13.8084, shapeName=pelvis_capsule, center=<0 -4.0604 0>, rotation=<rot 0 -90 0>}]
 spine_01 [{shapeType=Capsule, length=15.3859, radius=11.6369, shapeName=spine_01_capsule, center=<0 -3.3068 0>, rotation=<rot 0 -90 0>}]
 spine_02 [{shapeType=Capsule, length=13.9002, radius=8.0998, shapeName=spine_02_capsule, center=<0 -3.6362 0>, rotation=<rot 0 -90 0>}]
 spine_03 [{shapeType=Capsule, length=7.728, radius=8.0998, shapeName=spine_03_capsule, center=<0 -1.7019 0>, rotation=<rot 0 -90 0>}]
 spine_04 [{shapeType=Capsule, length=5.4289, radius=8.0998, shapeName=spine_04_capsule, rotation=<rot 0 -90 0>}]
 spine_05 [
  {shapeType=Capsule, length=5.4289, radius=8.97, shapeName=spine_05_capsule, center=<0.8201 1.0297 0.0546>, rotation=<rot 0 -90 0>}
  {shapeType=Capsule, length=60.2873, radius=15, shapeName=spine_05_capsule, center=<10.0313 13.0535 0>}
 ]
 upperarm_l [
  {
   shapeType=Capsule
   length=24.1538
   radius=3.8365
   shapeName=upperarm_l_capsule
   center=<11.8727 0 -0.5761>
   rotation=<rot -0.0001 -89.9999 92.7782>
  }
  {shapeType=Capsule, length=0.3759, radius=6.4145, shapeName=upperarm_l_capsule}
 ]
 upperarm_r [
  {
   shapeType=Capsule
   length=22.9588
   radius=3.8365
   shapeName=upperarm_r_capsule
   center=<-11.193 0.58 0.5432>
   rotation=<rot 0.0843 -91.7402 -87.2231>
  }
  {shapeType=Capsule, length=0.3759, radius=6.4145, shapeName=upperarm_r_capsule, center=<0 -0.122 0>, rotation=<rot 0 0 -180>}
 ]
}
constraints [
 {bone1Name=dyn_hair_08_l, bone2Name=dyn_hair_07_l, swing1LimitDegrees=8, swing2LimitDegrees=2, twistLimitDegrees=45, twistMotion=Locked}
 {bone1Name=dyn_hair_02_l, bone2Name=dyn_hair_01_l, swing1LimitDegrees=8, swing2LimitDegrees=2, twistLimitDegrees=45, twistMotion=Locked}
 {bone1Name=dyn_hair_03_l, bone2Name=dyn_hair_02_l, swing1LimitDegrees=8, swing2LimitDegrees=2, twistLimitDegrees=45, twistMotion=Locked}
 {bone1Name=dyn_hair_04_l, bone2Name=dyn_hair_03_l, swing1LimitDegrees=8, swing2LimitDegrees=2, twistLimitDegrees=45, twistMotion=Locked}
 {bone1Name=dyn_hair_05_l, bone2Name=dyn_hair_04_l, swing1LimitDegrees=8, swing2LimitDegrees=2, twistLimitDegrees=45, twistMotion=Locked}
 {bone1Name=dyn_hair_06_l, bone2Name=dyn_hair_05_l, swing1LimitDegrees=8, swing2LimitDegrees=2, twistLimitDegrees=45, twistMotion=Locked}
 {bone1Name=dyn_hair_07_l, bone2Name=dyn_hair_06_l, swing1LimitDegrees=8, swing2LimitDegrees=2, twistLimitDegrees=45, twistMotion=Locked}
 {bone1Name=dyn_hair_02_r, bone2Name=dyn_hair_01_r, swing1LimitDegrees=8, swing2LimitDegrees=2, twistLimitDegrees=45, twistMotion=Locked}
 {bone1Name=dyn_hair_03_r, bone2Name=dyn_hair_02_r, swing1LimitDegrees=8, swing2LimitDegrees=2, twistLimitDegrees=45, twistMotion=Locked}
 {bone1Name=dyn_hair_04_r, bone2Name=dyn_hair_03_r, swing1LimitDegrees=8, swing2LimitDegrees=2, twistLimitDegrees=45, twistMotion=Locked}
 {bone1Name=dyn_hair_05_r, bone2Name=dyn_hair_04_r, swing1LimitDegrees=8, swing2LimitDegrees=2, twistLimitDegrees=45, twistMotion=Locked}
 {bone1Name=dyn_hair_06_r, bone2Name=dyn_hair_05_r, swing1LimitDegrees=8, swing2LimitDegrees=2, twistLimitDegrees=45, twistMotion=Locked}
 {bone1Name=dyn_hair_07_r, bone2Name=dyn_hair_06_r, swing1LimitDegrees=8, swing2LimitDegrees=2, twistLimitDegrees=45, twistMotion=Locked}
 {bone1Name=dyn_hair_08_r, bone2Name=dyn_hair_07_r, swing1LimitDegrees=8, swing2LimitDegrees=2, twistLimitDegrees=45, twistMotion=Locked}
 {bone1Name=dyn_hair_01_l, bone2Name=base_hair_l, swing1LimitDegrees=5, swing2LimitDegrees=2, twistLimitDegrees=45, twistMotion=Locked}
 {bone1Name=dyn_hair_01_r, bone2Name=base_hair_r, swing1LimitDegrees=5, swing2LimitDegrees=2, twistLimitDegrees=45, twistMotion=Locked}
]
