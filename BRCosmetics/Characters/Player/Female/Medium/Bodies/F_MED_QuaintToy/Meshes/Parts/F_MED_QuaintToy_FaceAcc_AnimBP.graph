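# Post-copy hair animation for the QuaintToy female character's FaceAcc part, adding secondary motion to hair, strands and bows.
# The parent mesh pose runs through a Control Rig, then splits: constraint-driven hair for backbling/cape/no-RBAN, otherwise AnimDynamics, RigidBody and springs.
# Both paths feed strand constraints and final AnimDynamics for strands and bows.
# The update event sets gravity override, RBAN settings, the branch flag and skydiving-dependent alphas.
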AnimBlueprint /BRCosmetics/Characters/Player/Female/Medium/Bodies/F_MED_QuaintToy/Meshes/Parts/F_MED_QuaintToy_FaceAcc_AnimBP
target_skeleton=&/F_MED_QuaintToy_FaceAcc_Skeleton
graphs [
 AnimGraph {nodes [
   n1 = "Constraint - dyn_hair_02_l (PPPPPP)"!(
    ComponentPose=Local To Component(LocalPose=Use cached pose 'preDyn Cache'())
    Alpha=0.7
    ConstraintWeights_0=0.2836
    ConstraintWeights_1:float=0.2769
    ConstraintWeights_2:float=0.2616
    ConstraintWeights_3:float=0.164
    ConstraintWeights_4:float=0.0138
    ConstraintWeights_5:float=0.0001
   ) -> (Pose) {
    boneToModify {boneName=dyn_hair_02_l}
    constraintSetup [
     {targetBone {boneName=neck_01}}
     {targetBone {boneName=neck_02}}
     {targetBone {boneName=head}}
     {targetBone {boneName=spine_05}}
     {targetBone {boneName=clavicle_l}}
     {targetBone {boneName=clavicle_r}}
    ]
    constraintWeights [1,1,1,1,1,1]
   }
   n2 = "Constraint - dyn_hair_03_l (PPP)"!(ComponentPose=@n1, Alpha=0.85, ConstraintWeights_0=0.7648, ConstraintWeights_1:float=0.0352, ConstraintWeights_2:float=0.2) -> (Pose) {
    boneToModify {boneName=dyn_hair_03_l}
    constraintSetup [{targetBone {boneName=clavicle_l}}, {targetBone {boneName=spine_05}}, {targetBone {boneName=head}}]
    constraintWeights [1,1,1]
   }
   n3 = "Constraint - dyn_hair_04_l (PP)"!(ComponentPose=@n2, Alpha=1, ConstraintWeights_0=0.3241, ConstraintWeights_1:float=0.6759) -> (Pose) {
    boneToModify {boneName=dyn_hair_04_l}
    constraintSetup [{targetBone {boneName=clavicle_l}}, {targetBone {boneName=spine_05}}]
    constraintWeights [1,1]
   }
   n4 = "Constraint - dyn_hair_05_l (PP)"!(ComponentPose=@n3, Alpha=1, ConstraintWeights_0=0.3998, ConstraintWeights_1:float=0.6002) -> (Pose) {
    boneToModify {boneName=dyn_hair_05_l}
    constraintSetup [{targetBone {boneName=clavicle_l}}, {targetBone {boneName=spine_05}}]
    constraintWeights [1,1]
   }
   n5 = "Constraint - dyn_hair_06_l (PPP)"!(ComponentPose=@n4, Alpha=1, ConstraintWeights_0=0.053, ConstraintWeights_1:float=0.6695, ConstraintWeights_2:float=0.2775) -> (Pose) {
    boneToModify {boneName=dyn_hair_06_l}
    constraintSetup [{targetBone {boneName=clavicle_l}}, {targetBone {boneName=spine_04}}, {targetBone {boneName=spine_05}}]
    constraintWeights [1,1,1]
   }
   n6 = "Constraint - dyn_hair_07_l (PPP)"!(ComponentPose=@n5, Alpha=1, ConstraintWeights_0=0.0577, ConstraintWeights_1:float=0.7318, ConstraintWeights_2:float=0.2105) -> (Pose) {
    boneToModify {boneName=dyn_hair_07_l}
    constraintSetup [{targetBone {boneName=clavicle_l}}, {targetBone {boneName=spine_04}}, {targetBone {boneName=spine_05}}]
    constraintWeights [1,1,1]
   }
   n7 = "Constraint - dyn_hair_08_l (PPPPP)"!(
    ComponentPose=@n6
    Alpha=1
    ConstraintWeights_0=0.6089
    ConstraintWeights_1:float=0.2385
    ConstraintWeights_2:float=0.0621
    ConstraintWeights_3:float=0.0529
    ConstraintWeights_4:float=0.0376
   ) -> (Pose) {
    boneToModify {boneName=dyn_hair_08_l}
    constraintSetup [
     {targetBone {boneName=pelvis}}
     {targetBone {boneName=spine_02}}
     {targetBone {boneName=spine_01}}
     {targetBone {boneName=spine_04}}
     {targetBone {boneName=spine_03}}
    ]
    constraintWeights [1,1,1,1,1]
   }
   n8 = "Constraint - dyn_hair_02_r (PPPPPP)"!(
    ComponentPose=@n7
    Alpha=0.7
    ConstraintWeights_0=0.2837
    ConstraintWeights_1:float=0.2766
    ConstraintWeights_2:float=0.261
    ConstraintWeights_3:float=0.1647
    ConstraintWeights_4:float=0.0138
    ConstraintWeights_5:float=0.0001
   ) -> (Pose) {
    boneToModify {boneName=dyn_hair_02_r}
    constraintSetup [
     {targetBone {boneName=neck_01}}
     {targetBone {boneName=neck_02}}
     {targetBone {boneName=head}}
     {targetBone {boneName=spine_05}}
     {targetBone {boneName=clavicle_r}}
     {targetBone {boneName=clavicle_l}}
    ]
    constraintWeights [1,1,1,1,1,1]
   }
   n9 = "Constraint - dyn_hair_03_r (PPP)"!(ComponentPose=@n8, Alpha=0.85, ConstraintWeights_0=0.7305, ConstraintWeights_1:float=0.0695, ConstraintWeights_2:float=0.2) -> (Pose) {
    boneToModify {boneName=dyn_hair_03_r}
    constraintSetup [{targetBone {boneName=clavicle_r}}, {targetBone {boneName=spine_05}}, {targetBone {boneName=head}}]
    constraintWeights [1,1,1]
   }
   n10 = "Constraint - dyn_hair_04_r (PPP)"!(ComponentPose=@n9, Alpha=1, ConstraintWeights_0=0.2845, ConstraintWeights_1:float=0.7039, ConstraintWeights_2:float=0.0115) -> (Pose) {
    boneToModify {boneName=dyn_hair_04_r}
    constraintSetup [{targetBone {boneName=clavicle_r}}, {targetBone {boneName=spine_05}}, {targetBone {boneName=spine_04}}]
    constraintWeights [1,1,1]
   }
   n11 = "Constraint - dyn_hair_05_r (PPP)"!(ComponentPose=@n10, Alpha=1, ConstraintWeights_0=0.2809, ConstraintWeights_1:float=0.621, ConstraintWeights_2:float=0.0981) -> (Pose) {
    boneToModify {boneName=dyn_hair_05_r}
    constraintSetup [{targetBone {boneName=clavicle_r}}, {targetBone {boneName=spine_05}}, {targetBone {boneName=spine_04}}]
    constraintWeights [1,1,1]
   }
   n12 = "Constraint - dyn_hair_06_r (PPP)"!(ComponentPose=@n11, Alpha=1, ConstraintWeights_0=0.1077, ConstraintWeights_1:float=0.6739, ConstraintWeights_2:float=0.2184) -> (Pose) {
    boneToModify {boneName=dyn_hair_06_r}
    constraintSetup [{targetBone {boneName=clavicle_r}}, {targetBone {boneName=spine_04}}, {targetBone {boneName=spine_05}}]
    constraintWeights [1,1,1]
   }
   n13 = "Constraint - dyn_hair_07_r (PPP)"!(ComponentPose=@n12, Alpha=1, ConstraintWeights_0=0.0019, ConstraintWeights_1:float=0.7809, ConstraintWeights_2:float=0.2172) -> (Pose) {
    boneToModify {boneName=dyn_hair_07_r}
    constraintSetup [{targetBone {boneName=clavicle_r}}, {targetBone {boneName=spine_04}}, {targetBone {boneName=spine_05}}]
    constraintWeights [1,1,1]
   }
   n14 = "Constraint - dyn_hair_08_r (PPPPP)"!(
    ComponentPose=@n13
    Alpha=1
    ConstraintWeights_0=0.6158
    ConstraintWeights_1:float=0.2501
    ConstraintWeights_2:float=0.0652
    ConstraintWeights_3:float=0.0391
    ConstraintWeights_4:float=0.0298
   ) -> (Pose) {
    boneToModify {boneName=dyn_hair_08_r}
    constraintSetup [
     {targetBone {boneName=pelvis}}
     {targetBone {boneName=spine_02}}
     {targetBone {boneName=spine_01}}
     {targetBone {boneName=spine_04}}
     {targetBone {boneName=spine_03}}
    ]
    constraintWeights [1,1,1,1,1]
   }
   n15 = "Transform (Modify) Bone - Bone: dyn_hair_03_l"(ComponentPose=@n14, Alpha=1, Translation="0.000,2.815,0.000", Rotation="0.012,3.204,0.084", Scale="1,1,1") {
    rotationMode=BMM_Additive
    rotationSpace=BCS_BoneSpace
    translationMode=BMM_Additive
    translationSpace=BCS_BoneSpace
    boneToModify {boneName=dyn_hair_03_l}
    rotation=<rot 0.012 3.204 0.084>
    translation {y=2.815}
   }
   n16 = "Transform (Modify) Bone - Bone: dyn_hair_03_r"(ComponentPose=@n15, Alpha=1, Translation="0.000,1.898,0.000", Rotation="0.114,3.2,0.141", Scale="1,1,1") {
    rotationMode=BMM_Additive
    rotationSpace=BCS_BoneSpace
    translationMode=BMM_Additive
    translationSpace=BCS_BoneSpace
    boneToModify {boneName=dyn_hair_03_r}
    rotation=<rot 0.114 3.2 0.141>
    translation {y=1.898}
   }
   n17 = Component To Local(ComponentPose=@n16)
   n18 = Anim Dynamics - Bone: dyn_hair_01_l!(
    ComponentPose=Local To Component(LocalPose=Use cached pose 'preDyn Cache'())
    Alpha=1
    GravityOverride:Vector="0,0,-980"
    ExternalForce="0,0,0"
   ) -> (Pose) {
    angularDampingOverride=1
    angularSpringConstant=2000
    bAngularSpring=true
    bOverrideAngularDamping=true
    bUseGravityOverride=true
    gravityScale=0.35
    simulationSpace=RootRelative
    alphaScaleBiasClamp {bInterpResult=true, interpSpeedDecreasing=5}
    boundBone {boneName=dyn_hair_01_l}
    componentLinearAccScale=<0.3 0.3 0.7>
    componentLinearVelScale=<0.3 0.3 0.7>
    gravityOverride {z=-980}
    physicsBodyDefinitions [{
      boundBone {boneName=dyn_hair_01_l}
      boxExtents=<30 25 25>
      constraintSetup {angularLimitsMax=<0 1 2>, angularLimitsMin=<0 -25 -45>, angularTarget=<1 0 0>}
      localJointOffset=<4 5 6>
     }]
    simSpaceSettings {maxAngularAcceleration=3000, maxAngularVelocity=3000, simSpaceAngularAlpha=0.1}
   }
   n19 = Anim Dynamics - Bone: dyn_hair_01_r!(ComponentPose=@n18, Alpha=1, GravityOverride:Vector="0,0,-980", ExternalForce="0,0,0") -> (Pose) {
    angularDampingOverride=1
    angularSpringConstant=4000
    bAngularSpring=true
    bOverrideAngularDamping=true
    bUseGravityOverride=true
    gravityScale=0.35
    simulationSpace=RootRelative
    alphaScaleBiasClamp {bInterpResult=true, interpSpeedDecreasing=5}
    boundBone {boneName=dyn_hair_01_r}
    componentLinearAccScale=<0.3 0.3 0.7>
    componentLinearVelScale=<0.3 0.3 0.7>
    gravityOverride {z=-980}
    physicsBodyDefinitions [{
      boundBone {boneName=dyn_hair_01_r}
      boxExtents=<20 25 25>
      constraintSetup {angularLimitsMax=<0 25 2>, angularLimitsMin=<0 -1 -35>, angularTarget=<1 0 0>}
      localJointOffset=<5 2 -2>
     }]
    simSpaceSettings {maxAngularAcceleration=3000, maxAngularVelocity=3000, simSpaceAngularAlpha=0.1}
   }
   n20 = RigidBody!(
    ComponentPose=@n19
    Alpha
    OverrideWorldGravity:Vector="0,0,0"
    ExternalForce
    SimSpaceSettings:Sim Space Settings Structure=(WorldAlpha=0.6,VelocityScaleZ=0.7,DampingAlpha=0,MaxLinearVelocity=100,MaxAngularVelocity=5,MaxLinearAcceleration=300,MaxAngularAcceleration=12,ExternalLinearDragV=(X=0.2,Y=0.2,Z=0.2),ExternalLinearVelocity=(X=0,Y=0,Z=0),ExternalAngularVelocity=(X=0,Y=0,Z=0))
   ) -> (Pose) {
    bOverrideWorldGravity=true
    overridePhysicsAsset=&/F_MED_QuaintToy_FaceAcc_Physics
    simulationSpace=BaseBoneSpace
    baseBoneRef {boneName=spine_03}
    componentAppliedLinearAccClamp=<3000 3000 3000>
    simSpaceSettings {
     dampingAlpha=0
     maxAngularAcceleration=12
     maxAngularVelocity=5
     maxLinearAcceleration=300
     maxLinearVelocity=100
     velocityScaleZ=0.7
     worldAlpha=0.6
     externalLinearDragV=<0.2 0.2 0.2>
    }
   }
   n21 = Spring controller - Bone: dyn_hairSpringJnt_03_l!(ComponentPose:Component Space Pose Link Structure=@n20, Alpha:float=0.02) -> (Pose:Component Space Pose Link Structure) {
    is=title AnimGraphNode_SpringBone
    alpha=0.02
    bTranslateX=false
    bTranslateY=false
    maxDisplacement=5
    springDamping=16
    springStiffness=300
    alphaScaleBias {scale=0.6667}
    springBone {boneName=dyn_hairSpringJnt_03_l}
   }
   n22 = Spring controller - Bone: dyn_hairSpringJnt_04_l!(ComponentPose:Component Space Pose Link Structure=@n21, Alpha:float=0.02) -> (Pose:Component Space Pose Link Structure) {
    is=title AnimGraphNode_SpringBone
    alpha=0.02
    bTranslateX=false
    bTranslateY=false
    maxDisplacement=5
    springDamping=16
    springStiffness=300
    alphaScaleBias {scale=0.6667}
    springBone {boneName=dyn_hairSpringJnt_04_l}
   }
   n23 = Spring controller - Bone: dyn_hairSpringJnt_05_l!(ComponentPose:Component Space Pose Link Structure=@n22, Alpha:float=0.05) -> (Pose:Component Space Pose Link Structure) {
    is=title AnimGraphNode_SpringBone
    alpha=0.05
    bTranslateX=false
    bTranslateY=false
    maxDisplacement=5
    springDamping=16
    springStiffness=300
    springBone {boneName=dyn_hairSpringJnt_05_l}
   }
   n24 = Spring controller - Bone: dyn_hairSpringJnt_06_l!(ComponentPose:Component Space Pose Link Structure=@n23, Alpha:float=0.02) -> (Pose:Component Space Pose Link Structure) {
    is=title AnimGraphNode_SpringBone
    alpha=0.02
    bTranslateX=false
    bTranslateY=false
    maxDisplacement=5
    springDamping=16
    springStiffness=300
    alphaScaleBias {scale=0.6667}
    springBone {boneName=dyn_hairSpringJnt_06_l}
   }
   n25 = Spring controller - Bone: dyn_hairSpringJnt_07_l!(ComponentPose:Component Space Pose Link Structure=@n24, Alpha:float=0.05) -> (Pose:Component Space Pose Link Structure) {
    is=title AnimGraphNode_SpringBone
    alpha=0.05
    bTranslateX=false
    bTranslateY=false
    maxDisplacement=5
    springDamping=16
    springStiffness=300
    springBone {boneName=dyn_hairSpringJnt_07_l}
   }
   n26 = Spring controller - Bone: dyn_hairSpringJnt_08_l!(ComponentPose:Component Space Pose Link Structure=@n25, Alpha:float=0.05) -> (Pose:Component Space Pose Link Structure) {
    is=title AnimGraphNode_SpringBone
    alpha=0.05
    bTranslateX=false
    bTranslateY=false
    maxDisplacement=5
    springDamping=16
    springStiffness=300
    springBone {boneName=dyn_hairSpringJnt_08_l}
   }
   n27 = Spring controller - Bone: dyn_hairSpringJnt_03_r!(ComponentPose:Component Space Pose Link Structure=@n26, Alpha:float=1) -> (Pose:Component Space Pose Link Structure) {
    is=title AnimGraphNode_SpringBone
    alpha=0.02
    bTranslateX=false
    bTranslateY=false
    maxDisplacement=5
    springDamping=16
    springStiffness=300
    alphaScaleBias {scale=0.6667}
    springBone {boneName=dyn_hairSpringJnt_03_r}
   }
   n28 = Spring controller - Bone: dyn_hairSpringJnt_04_r!(ComponentPose:Component Space Pose Link Structure=@n27, Alpha:float=1) -> (Pose:Component Space Pose Link Structure) {
    is=title AnimGraphNode_SpringBone
    alpha=0.02
    bTranslateX=false
    bTranslateY=false
    maxDisplacement=5
    springDamping=16
    springStiffness=300
    alphaScaleBias {scale=0.6667}
    springBone {boneName=dyn_hairSpringJnt_04_r}
   }
   n29 = Spring controller - Bone: dyn_hairSpringJnt_05_r!(ComponentPose:Component Space Pose Link Structure=@n28, Alpha:float=0.05) -> (Pose:Component Space Pose Link Structure) {
    is=title AnimGraphNode_SpringBone
    alpha=0.05
    bTranslateX=false
    bTranslateY=false
    maxDisplacement=5
    springDamping=16
    springStiffness=300
    springBone {boneName=dyn_hairSpringJnt_05_r}
   }
   n30 = Spring controller - Bone: dyn_hairSpringJnt_06_r!(ComponentPose:Component Space Pose Link Structure=@n29, Alpha:float=1) -> (Pose:Component Space Pose Link Structure) {
    is=title AnimGraphNode_SpringBone
    alpha=0.02
    bTranslateX=false
    bTranslateY=false
    maxDisplacement=5
    springDamping=16
    springStiffness=300
    alphaScaleBias {scale=0.6667}
    springBone {boneName=dyn_hairSpringJnt_06_r}
   }
   n31 = Spring controller - Bone: dyn_hairSpringJnt_07_r!(ComponentPose:Component Space Pose Link Structure=@n30, Alpha:float=0.05) -> (Pose:Component Space Pose Link Structure) {
    is=title AnimGraphNode_SpringBone
    alpha=0.05
    bTranslateX=false
    bTranslateY=false
    maxDisplacement=5
    springDamping=16
    springStiffness=300
    springBone {boneName=dyn_hairSpringJnt_07_r}
   }
   n32 = Spring controller - Bone: dyn_hairSpringJnt_08_r!(ComponentPose:Component Space Pose Link Structure=@n31, Alpha:float=0.05) -> (Pose:Component Space Pose Link Structure) {
    is=title AnimGraphNode_SpringBone
    alpha=0.05
    bTranslateX=false
    bTranslateY=false
    maxDisplacement=5
    springDamping=16
    springStiffness=300
    springBone {boneName=dyn_hairSpringJnt_08_r}
   }
   n33 = Component To Local(ComponentPose=@n32)
   n34 = Get isWearingBackbling / Cape / NoRBAN() -> (isWearingBackbling / Cape / NoRBAN:bool)
   n35 = "Constraint - base_hairStrand_C_01_r (PPP)"!(
    ComponentPose=Local To Component(LocalPose=Blend Poses by bool(BlendPose_0=@n17, BlendPose_1=@n33, bActiveValue=@n34))
    Alpha=1
    ConstraintWeights_0=0.5053
    ConstraintWeights_1:float=0.3104
    ConstraintWeights_2:float=0.1843
   ) -> (Pose) {
    boneToModify {boneName=base_hairStrand_C_01_r}
    constraintSetup [
     {targetBone {boneName=dyn_hairSpringJnt_07_r}}
     {targetBone {boneName=dyn_hairSpringJnt_08_r}}
     {targetBone {boneName=dyn_hairSpringJnt_06_r}}
    ]
    constraintWeights [1,1,1]
   }
   n36 = "Constraint - base_hairStrand_B_01_l (PPP)"!(ComponentPose=@n35, Alpha=1, ConstraintWeights_0=0.5193, ConstraintWeights_1:float=0.2931, ConstraintWeights_2:float=0.1877) -> (Pose) {
    boneToModify {boneName=base_hairStrand_B_01_l}
    constraintSetup [
     {targetBone {boneName=dyn_hairSpringJnt_07_l}}
     {targetBone {boneName=dyn_hairSpringJnt_08_l}}
     {targetBone {boneName=dyn_hairSpringJnt_06_l}}
    ]
    constraintWeights [1,1,1]
   }
   n37 = "Constraint - base_hairStrand_B_01_r (PPPPPP)"!(
    ComponentPose=@n36
    Alpha=1
    ConstraintWeights_0=0.5262
    ConstraintWeights_1:float=0.4436
    ConstraintWeights_2:float=0.0274
    ConstraintWeights_3:float=0.0024
    ConstraintWeights_4:float=0.0002
    ConstraintWeights_5:float=0.0001
   ) -> (Pose) {
    boneToModify {boneName=base_hairStrand_B_01_r}
    constraintSetup [
     {targetBone {boneName=dyn_hairSpringJnt_06_r}}
     {targetBone {boneName=dyn_hairSpringJnt_07_r}}
     {targetBone {boneName=dyn_hairSpringJnt_05_r}}
     {targetBone {boneName=dyn_hairSpringJnt_08_r}}
     {targetBone {boneName=dyn_hairSpringJnt_04_r}}
     {targetBone {boneName=dyn_hairSpringJnt_03_r}}
    ]
    constraintWeights [1,1,1,1,1,1]
   }
   n38 = "Constraint - base_hairStrand_A_01_r (PPPPPPP)"!(
    ComponentPose=@n37
    Alpha=1
    ConstraintWeights_0=0.3365
    ConstraintWeights_1:float=0.3147
    ConstraintWeights_2:float=0.2331
    ConstraintWeights_3:float=0.065
    ConstraintWeights_4:float=0.0466
    ConstraintWeights_5:float=0.0035
    ConstraintWeights_6:float=0.0007
   ) -> (Pose) {
    boneToModify {boneName=base_hairStrand_A_01_r}
    constraintSetup [
     {targetBone {boneName=dyn_hairSpringJnt_04_r}}
     {targetBone {boneName=dyn_hairSpringJnt_03_r}}
     {targetBone {boneName=dyn_hairStrand_A_01_r}}
     {targetBone {boneName=dyn_hair_02_r}}
     {targetBone {boneName=dyn_hairSpringJnt_05_r}}
     {targetBone {boneName=dyn_hairStrand_A_02_r}}
     {targetBone {boneName=dyn_hair_01_r}}
    ]
    constraintWeights [1,1,1,1,1,1,1]
   }
   n39 = "Constraint - base_hairStrand_A_01_l (PPPPPPP)"!(
    ComponentPose=@n38
    Alpha=1
    ConstraintWeights_0=0.5589
    ConstraintWeights_1:float=0.2306
    ConstraintWeights_2:float=0.1118
    ConstraintWeights_3:float=0.0905
    ConstraintWeights_4:float=0.0055
    ConstraintWeights_5:float=0.0025
    ConstraintWeights_6:float=0.0002
   ) -> (Pose) {
    boneToModify {boneName=base_hairStrand_A_01_l}
    constraintSetup [
     {targetBone {boneName=dyn_hairSpringJnt_04_l}}
     {targetBone {boneName=dyn_hairStrand_A_01_l}}
     {targetBone {boneName=dyn_hairSpringJnt_05_l}}
     {targetBone {boneName=dyn_hairSpringJnt_03_l}}
     {targetBone {boneName=dyn_hair_02_l}}
     {targetBone {boneName=dyn_hairStrand_A_02_l}}
     {targetBone {boneName=dyn_hairSpringJnt_06_l}}
    ]
    constraintWeights [1,1,1,1,1,1,1]
   }
   n40 = Anim Dynamics - Bone: base_hairStrand_A_01_l!(ComponentPose=@n39, GravityOverride:Vector="0,0,-980", ExternalForce="0,0,0") -> (Pose) {
    angularDampingOverride=0.999
    angularSpringConstant=7000
    bAngularSpring=true
    bOverrideAngularDamping=true
    bUseGravityOverride=true
    gravityScale=0.35
    simulationSpace=RootRelative
    boundBone {boneName=base_hairStrand_A_01_l}
    componentAppliedLinearAccClamp=<3000 3000 3000>
    componentLinearAccScale=<0.3 0.3 0.3>
    componentLinearVelScale=<0.3 0.3 0.3>
    gravityOverride {z=-980}
    physicsBodyDefinitions [{
      boundBone {boneName=base_hairStrand_A_01_l}
      boxExtents=<40 40 40>
      constraintSetup {angularLimitsMax=<0 8 1>, angularLimitsMin=<0 -8 -10>, angularTarget=<1 0 0>}
      localJointOffset=<-1 0 0>
     }]
    relativeSpaceBone {boneName=spine_05}
    simSpaceSettings {simSpaceAngularAlpha=0.1}
   }
   n41 = Anim Dynamics - Bone: dyn_hairStrand_A_01_r!(ComponentPose=@n40, GravityOverride:Vector="0,0,-980", ExternalForce="0,0,0") -> (Pose) {
    angularDampingOverride=0.999
    angularSpringConstant=7000
    bAngularSpring=true
    bOverrideAngularDamping=true
    bUseGravityOverride=true
    gravityScale=0.35
    simulationSpace=RootRelative
    boundBone {boneName=dyn_hairStrand_A_01_r}
    componentAppliedLinearAccClamp=<3000 3000 3000>
    componentLinearAccScale=<0.3 0.3 0.3>
    componentLinearVelScale=<0.3 0.3 0.3>
    gravityOverride {z=-980}
    physicsBodyDefinitions [{
      boundBone {boneName=dyn_hairStrand_A_01_r}
      boxExtents=<40 40 40>
      constraintSetup {angularLimitsMax=<0 8 5>, angularLimitsMin=<0 -8 -10>, angularTarget=<1 0 0>}
      localJointOffset=<-1 0 0>
     }]
    relativeSpaceBone {boneName=spine_05}
    simSpaceSettings {simSpaceAngularAlpha=0.1}
   }
   n42 = Anim Dynamics - Bone: dyn_hairStrand_A_02_l!(ComponentPose=@n41, GravityOverride:Vector="0,0,-980", ExternalForce="0,0,0") -> (Pose) {
    angularDampingOverride=0.999
    angularSpringConstant=7000
    bAngularSpring=true
    bOverrideAngularDamping=true
    bUseGravityOverride=true
    gravityScale=0.35
    simulationSpace=RootRelative
    boundBone {boneName=dyn_hairStrand_A_02_l}
    componentAppliedLinearAccClamp=<3000 3000 3000>
    componentLinearAccScale=<0.3 0.3 0.3>
    componentLinearVelScale=<0.3 0.3 0.3>
    gravityOverride {z=-980}
    physicsBodyDefinitions [{
      boundBone {boneName=dyn_hairStrand_A_02_l}
      boxExtents=<40 40 40>
      constraintSetup {angularLimitsMax=<0 8 1>, angularLimitsMin=<0 -8 -10>, angularTarget=<1 0 0>}
      localJointOffset=<-1 0 0>
     }]
    relativeSpaceBone {boneName=spine_05}
    simSpaceSettings {simSpaceAngularAlpha=0.1}
   }
   n43 = Anim Dynamics - Bone: dyn_hairStrand_A_02_r!(ComponentPose=@n42, GravityOverride:Vector="0,0,-980", ExternalForce="0,0,0") -> (Pose) {
    angularDampingOverride=0.999
    angularSpringConstant=7000
    bAngularSpring=true
    bOverrideAngularDamping=true
    bUseGravityOverride=true
    gravityScale=0.35
    simulationSpace=RootRelative
    boundBone {boneName=dyn_hairStrand_A_02_r}
    componentAppliedLinearAccClamp=<3000 3000 3000>
    componentLinearAccScale=<0.3 0.3 0.3>
    componentLinearVelScale=<0.3 0.3 0.3>
    gravityOverride {z=-980}
    physicsBodyDefinitions [{
      boundBone {boneName=dyn_hairStrand_A_02_r}
      boxExtents=<40 40 40>
      constraintSetup {angularLimitsMax=<0 8 5>, angularLimitsMin=<0 -8 -10>, angularTarget=<1 0 0>}
      localJointOffset=<-1 0 0>
     }]
    relativeSpaceBone {boneName=spine_05}
    simSpaceSettings {simSpaceAngularAlpha=0.1}
   }
   n44 = Anim Dynamics - Bone: dyn_hairStrand_A_03_l!(ComponentPose=@n43, GravityOverride:Vector="0,0,-980", ExternalForce="0,0,0") -> (Pose) {
    angularDampingOverride=0.999
    angularSpringConstant=7000
    bAngularSpring=true
    bOverrideAngularDamping=true
    bUseGravityOverride=true
    gravityScale=0.35
    simulationSpace=RootRelative
    boundBone {boneName=dyn_hairStrand_A_03_l}
    componentAppliedLinearAccClamp=<3000 3000 3000>
    componentLinearAccScale=<0.3 0.3 0.3>
    componentLinearVelScale=<0.3 0.3 0.3>
    gravityOverride {z=-980}
    physicsBodyDefinitions [{
      boundBone {boneName=dyn_hairStrand_A_03_l}
      boxExtents=<40 40 40>
      constraintSetup {angularLimitsMax=<0 8 1>, angularLimitsMin=<0 -8 -10>, angularTarget=<1 0 0>}
      localJointOffset=<-1 0 0>
     }]
    relativeSpaceBone {boneName=spine_05}
    simSpaceSettings {simSpaceAngularAlpha=0.1}
   }
   n45 = Anim Dynamics - Bone: dyn_hairStrand_A_03_r!(ComponentPose=@n44, GravityOverride:Vector="0,0,-980", ExternalForce="0,0,0") -> (Pose) {
    angularDampingOverride=0.999
    angularSpringConstant=7000
    bAngularSpring=true
    bOverrideAngularDamping=true
    bUseGravityOverride=true
    gravityScale=0.35
    simulationSpace=RootRelative
    boundBone {boneName=dyn_hairStrand_A_03_r}
    componentAppliedLinearAccClamp=<3000 3000 3000>
    componentLinearAccScale=<0.3 0.3 0.3>
    componentLinearVelScale=<0.3 0.3 0.3>
    gravityOverride {z=-980}
    physicsBodyDefinitions [{
      boundBone {boneName=dyn_hairStrand_A_03_r}
      boxExtents=<40 40 40>
      constraintSetup {angularLimitsMax=<0 8 5>, angularLimitsMin=<0 -8 -10>, angularTarget=<1 0 0>}
      localJointOffset=<-1 0 0>
     }]
    relativeSpaceBone {boneName=spine_05}
    simSpaceSettings {simSpaceAngularAlpha=0.1}
   }
   n46 = Anim Dynamics - Bone: dyn_hairStrand_B_01_l!(ComponentPose=@n45, GravityOverride:Vector="0,0,-980", ExternalForce="0,0,0") -> (Pose) {
    angularDampingOverride=0.999
    angularSpringConstant=7000
    bAngularSpring=true
    bOverrideAngularDamping=true
    bUseGravityOverride=true
    gravityScale=0.35
    simulationSpace=RootRelative
    boundBone {boneName=dyn_hairStrand_B_01_l}
    componentAppliedLinearAccClamp=<3000 3000 3000>
    componentLinearAccScale=<0.4 0.4 0.5>
    componentLinearVelScale=<0.4 0.4 0.5>
    gravityOverride {z=-980}
    physicsBodyDefinitions [{
      boundBone {boneName=dyn_hairStrand_B_01_l}
      boxExtents=<20 20 20>
      constraintSetup {angularLimitsMax=<0 0 15>, angularLimitsMin=<0 0 -15>, angularTarget=<1 0 0>}
      localJointOffset=<-1 1 0>
     }]
    relativeSpaceBone {boneName=spine_05}
    simSpaceSettings {simSpaceAngularAlpha=0.1}
   }
   n47 = Anim Dynamics - Bone: dyn_hairStrand_C_01_r!(ComponentPose=@n46, GravityOverride:Vector="0,0,-980", ExternalForce="0,0,0") -> (Pose) {
    angularDampingOverride=0.999
    angularSpringConstant=7000
    bAngularSpring=true
    bOverrideAngularDamping=true
    bUseGravityOverride=true
    gravityScale=0.35
    simulationSpace=RootRelative
    boundBone {boneName=dyn_hairStrand_C_01_r}
    componentAppliedLinearAccClamp=<3000 3000 3000>
    componentLinearAccScale=<0.4 0.4 0.5>
    componentLinearVelScale=<0.4 0.4 0.5>
    gravityOverride {z=-980}
    physicsBodyDefinitions [{
      boundBone {boneName=dyn_hairStrand_C_01_r}
      boxExtents=<20 20 20>
      constraintSetup {angularLimitsMax=<0 0 15>, angularLimitsMin=<0 0 -15>, angularTarget=<1 0 0>}
      localJointOffset=<-1 1 0>
     }]
    relativeSpaceBone {boneName=spine_05}
    simSpaceSettings {simSpaceAngularAlpha=0.1}
   }
   n48 = Anim Dynamics - Bone: dyn_hairStrand_B_01_r!(ComponentPose=@n47, GravityOverride:Vector="0,0,-980", ExternalForce="0,0,0") -> (Pose) {
    angularDampingOverride=0.999
    angularSpringConstant=7000
    bAngularSpring=true
    bOverrideAngularDamping=true
    bUseGravityOverride=true
    gravityScale=0.35
    simulationSpace=RootRelative
    boundBone {boneName=dyn_hairStrand_B_01_r}
    componentAppliedLinearAccClamp=<3000 3000 3000>
    componentLinearAccScale=<0.4 0.4 0.5>
    componentLinearVelScale=<0.4 0.4 0.5>
    gravityOverride {z=-980}
    physicsBodyDefinitions [{
      boundBone {boneName=dyn_hairStrand_B_01_r}
      boxExtents=<20 20 20>
      constraintSetup {angularLimitsMax=<0 0 15>, angularLimitsMin=<0 0 -15>, angularTarget=<1 0 0>}
      localJointOffset=<-1 1 0>
     }]
    relativeSpaceBone {boneName=spine_05}
    simSpaceSettings {simSpaceAngularAlpha=0.1}
   }
   n49 = Anim Dynamics - Bone: dyn_hairBow_01_r!(ComponentPose=@n48, Alpha=0.5, GravityOverride:Vector="0,0,-980", ExternalForce="0,0,0") -> (Pose) {
    angularDampingOverride=0.999
    angularSpringConstant=7000
    bAngularSpring=true
    bOverrideAngularDamping=true
    bUseGravityOverride=true
    gravityScale=0.35
    simulationSpace=RootRelative
    boundBone {boneName=dyn_hairBow_01_r}
    componentAppliedLinearAccClamp=<3500 3500 3500>
    componentLinearAccScale=<0.4 0.4 0.5>
    componentLinearVelScale=<0.4 0.4 0.5>
    gravityOverride {z=-980}
    physicsBodyDefinitions [{
      boundBone {boneName=dyn_hairBow_01_r}
      boxExtents=<15 15 15>
      constraintSetup {angularLimitsMax=<0 5 2>, angularLimitsMin=<0 -5 -20>, angularTarget=<1 0 0>}
      localJointOffset=<1 0.5 0>
     }]
    relativeSpaceBone {boneName=spine_05}
    simSpaceSettings {simSpaceAngularAlpha=0.1}
   }
   n50 = Anim Dynamics - Bone: dyn_hairBow_02_r!(ComponentPose=@n49, Alpha=0.5, GravityOverride:Vector="0,0,-980", ExternalForce="0,0,0") -> (Pose) {
    angularDampingOverride=0.999
    angularSpringConstant=7000
    bAngularSpring=true
    bOverrideAngularDamping=true
    bUseGravityOverride=true
    gravityScale=0.35
    simulationSpace=RootRelative
    boundBone {boneName=dyn_hairBow_02_r}
    componentAppliedLinearAccClamp=<3500 3500 3500>
    componentLinearAccScale=<0.4 0.4 0.5>
    componentLinearVelScale=<0.4 0.4 0.5>
    gravityOverride {z=-980}
    physicsBodyDefinitions [{
      boundBone {boneName=dyn_hairBow_02_r}
      boxExtents=<15 15 15>
      constraintSetup {angularLimitsMax=<0 5 2>, angularLimitsMin=<0 -5 -20>, angularTarget=<1 0 0>}
      localJointOffset=<1 0.5 0>
     }]
    relativeSpaceBone {boneName=spine_05}
    simSpaceSettings {simSpaceAngularAlpha=0.1}
   }
   n51 = Anim Dynamics - Bone: dyn_hairBow_01_l!(ComponentPose=@n50, Alpha=0.5, GravityOverride:Vector="0,0,-980", ExternalForce="0,0,0") -> (Pose) {
    angularDampingOverride=0.999
    angularSpringConstant=7000
    bAngularSpring=true
    bOverrideAngularDamping=true
    bUseGravityOverride=true
    gravityScale=0.35
    simulationSpace=RootRelative
    boundBone {boneName=dyn_hairBow_01_l}
    componentAppliedLinearAccClamp=<3500 3500 3500>
    componentLinearAccScale=<0.4 0.4 0.5>
    componentLinearVelScale=<0.4 0.4 0.5>
    gravityOverride {z=-980}
    physicsBodyDefinitions [{
      boundBone {boneName=dyn_hairBow_01_l}
      boxExtents=<15 15 15>
      constraintSetup {angularLimitsMax=<0 5 2>, angularLimitsMin=<0 -5 -20>, angularTarget=<1 0 0>}
      localJointOffset=<1 0.5 0>
     }]
    relativeSpaceBone {boneName=spine_05}
    simSpaceSettings {simSpaceAngularAlpha=0.1}
   }
   n52 = Anim Dynamics - Bone: dyn_hairBow_02_l!(ComponentPose=@n51, Alpha=0.5, GravityOverride:Vector="0,0,-980", ExternalForce="0,0,0") -> (Pose) {
    angularDampingOverride=0.999
    angularSpringConstant=7000
    bAngularSpring=true
    bOverrideAngularDamping=true
    bUseGravityOverride=true
    gravityScale=0.35
    simulationSpace=RootRelative
    boundBone {boneName=dyn_hairBow_02_l}
    componentAppliedLinearAccClamp=<3500 3500 3500>
    componentLinearAccScale=<0.4 0.4 0.5>
    componentLinearVelScale=<0.4 0.4 0.5>
    gravityOverride {z=-980}
    physicsBodyDefinitions [{
      boundBone {boneName=dyn_hairBow_02_l}
      boxExtents=<15 15 15>
      constraintSetup {angularLimitsMax=<0 5 2>, angularLimitsMin=<0 -5 -20>, angularTarget=<1 0 0>}
      localJointOffset=<1 0.5 0>
     }]
    relativeSpaceBone {boneName=spine_05}
    simSpaceSettings {simSpaceAngularAlpha=0.1}
   }
   Output Pose(Result=Component To Local(ComponentPose=@n52)) {layerGroup=None}
   n53 = Control Rig(Source=Copy Pose From Mesh() {bUseAttachedParent=true}) {controlRigAssetReference {controlRigAsset=&/F_MED_QuaintToy_FaceAcc_CtrlRig}}
   Save cached pose 'preDyn Cache'(Pose=@n53) {cacheName=preDyn Cache}
  ]}
 EventGraph {nodes [
   Event BlueprintUpdateAnimation()
   > n1 = "Gravity Override State Function (Parameter Asset) (Nativized)"(
    CharacterPartAnimInstance=Self-Reference()
    GameTimeInSeconds=GetGameTimeInSeconds()
    DebugJointName=spine_05
    Parameters=&/F_MED_QuaintToy_FaceAcc_Parameters
   )
   > Set Gravity Override(Gravity Override=@n1)
   > n2 = Set RBAN Settings Out(RBAN Settings Out=@n1.OutRigidBodyAnimNodeParameters)
   n3 = Get bRigidBodyEnabled()
   n4 = CheckForBackbling()
   Set isWearingBackbling / Cape / NoRBAN(
    execute=[@n6.OutputPin, @n2.then]
    isWearingBackbling / Cape / NoRBAN:bool=OR Boolean!(A=@n4, B=@n4.bIsBackblingACape, C:bool=@n4.bIsBackblingLong, D:bool=NOT Boolean(A=@n3)) -> (ReturnValue)
   ) -> (Output_Get:bool)
   > n5 = Branch(Condition=Get bIsSkyDiving())
   > Set springAlpha(springAlpha:double=0) -> (Output_Get:double)
   > Set rbanAlpha(rbanAlpha:double=0.6) -> (Output_Get:double)
   Set springAlpha(execute=@n5.else, springAlpha=0.12)
   > Set rbanAlpha(rbanAlpha=1)
   n6 = Reroute Node()
  ]}
]
variables [
 Gravity Override {type=Vector}
 isWearingBackbling / Cape / NoRBAN {type=boolean}
 RBAN Settings Out {type=FortRigidBodyAnimNodeParameters}
 rbanAlpha {type=number}
 springAlpha {type=number}
]
